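# Translucent UI material that draws the Fall Guys themed background for Discovery browser pages through MF_UI_Discovery_IP_Page.
# A screen-space blue gradient, shaded radially, is overlaid with faint anti-aliased zig-zag lines from MF_UI_Wave.
# The lines sit in rotated, repeating rows, using a ping-pong curve for shape and time-driven scrolling and thickness pulsing.
# Vertical position and a radial mask fade the pattern, and named reroutes carry shared values such as per-row randomness.
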Material /DiscoveryBrowser/Art/Materials/Backgrounds/M_UI_Discovery_Background_FallGuys
graph {
 properties {
  bAutomaticallySetUsageInEditor=false
  bUsedWithStaticMesh=false
  blendMode=BLEND_Translucent
  materialDomain=MD_UI
  usedShadingModels=Default Lit
 }
 nodes [
  Constant2Vector() {g=0.375}
  MaterialFunctionCall!(Background Color=NamedRerouteUsage()) {materialFunction=/DiscoveryBrowser/Art/MaterialFunctions/MF_UI_Discovery_IP_Page}
  n1 = Constant3Vector() {constant=<rgba 0.013 0.227 0.8879 1>}
  n2 = Constant3Vector() {constant=<rgba 0.0409 0.6724 0.9734 1>}
  n3 = ScreenPosition()
  n4 = ComponentMask(None=@n3.ViewportUV) {g=true}
  n5 = Constant2Vector() {g=0.5, r=0.5}
  n6 = MaterialFunctionCall(
   UVs=MaterialFunctionCall!(ScreenUV=Add(A=@n3.ViewportUV, B=@n5), TextureSize, ScalePivotPoint) {materialFunction=/FortUILibrary/MaterialFunctions/Basics/MF_UI_ScreenSpaceUVScaling}.ScaleY
   Texture Scale=Constant() {r=6}
  ) {materialFunction=&ScaleUVsByCenter}
  n7 = NamedRerouteDeclaration(None=LinearInterpolate(
    A=LinearInterpolate(A=@n1, B=@n2, Alpha=@n4)
    B=Constant3Vector() {constant=<rgba 0 0.0562 0.2135 0.8>}
    Alpha=MaterialFunctionCall(UV=@n6.UVs) {materialFunction=&MF_UI_RadialGradient}.Smooth
   )) {name=Gradient, nodeColor=<rgba 0 0 0 1>}
  n8 = ScreenPosition()
  n9 = Constant2Vector() {g=5, r=1}
  n10 = Rotator(
   Coordinate=MaterialFunctionCall(ScreenUV=Multiply(A=@n8.ViewportUV, B=@n9), TextureSize=@n9) {materialFunction=&MF_UI_ScreenSpaceUVScaling}.Scaled to Fit
   Time=NamedRerouteDeclaration(None=Constant() {r=-0.0675}) {name=Rotation Angle, nodeColor=<rgba 0.3125 1 0 1>}
  ) {speed=6.2832}
  n11 = AppendVector(A=ComponentMask(None=@n10) {r=true}, B=Frac(None=ComponentMask(None=@n10) {g=true}))
  n12 = Constant() {r=2}
  n13 = Time()
  n14 = LinearInterpolate(Alpha=Multiply(A=Multiply(A=@n13) {constB=0.15}, B=NamedRerouteUsage())) {constB=0.2}
  n15 = Constant() {r=0.375}
  n16 = NamedRerouteUsage()
  n17 = NamedRerouteUsage()
  n18 = Multiply(A=Time()) {constB=0.1}
  n19 = MaterialFunctionCall!(Value=Multiply(A=Max(A=@n16) {constB=0.1}, B=Add(A=@n17, B=@n18)), Scale=Constant() {r=4}, Height) {materialFunction=/FortUILibrary/MaterialFunctions/Basics/MF_UI_PingPong}
  n20 = MaterialFunctionCall!(
   UV=Add(
    A=MaterialFunctionCall(UVs=@n11, Texture Scale=@n12) {materialFunction=&ScaleUVsByCenter}.UVs
    B=AppendVector(A=@n14, B=@n15)
   )
   Wave Height=Constant() {r=0.25}
   Wave Elevation
   Wave Length=LinearInterpolate(A=Constant() {r=1.25}, B=Constant() {r=1.75}, Alpha=NamedRerouteUsage())
   Wave Speed=Constant()
   GlowMin
   GlowMax
   Outline Thickness=LinearInterpolate(Alpha=Saturate(None=@n19.Result)) {constA=0.12, constB=0.18}
   Outline Alignment
   Custom Curve=NamedRerouteUsage()
  ) {materialFunction=/FortUILibrary/MaterialFunctions/Shapes/MF_UI_Wave}
  n21 = NamedRerouteDeclaration(None=MaterialFunctionCall(DistanceField=@n20.SDF Outline) {materialFunction=&MF_UI_AntiAliasing}.Inner) {name=Zig Zag Mask, nodeColor=<rgba 0 0 0 1>}
  n22 = MaterialFunctionCall(Base=@n7, Blend=@n21) {materialFunction=&Blend_Overlay}
  n23 = Multiply(A=@n21) {constB=0.15}
  n24 = ScreenPosition()
  n25 = ComponentMask(None=@n24.ViewportUV) {g=true}
  n26 = Constant() {r=3}
  n27 = OneMinus(None=MaterialFunctionCall(UV=MaterialFunctionCall(UVs=@n24.ViewportUV, Texture Scale=@n26) {materialFunction=&ScaleUVsByCenter}.UVs) {materialFunction=&MF_UI_RadialGradient}.Smooth)
  NamedRerouteDeclaration(None=LinearInterpolate(A=LinearInterpolate(A=@n7, B=@n22.Result, Alpha=@n23), B=NamedRerouteUsage(), Alpha=LinearInterpolate(A=@n25, Alpha=@n27))) {name=Background, nodeColor=<rgba 0 0 0 1>}
  n28 = NamedRerouteUsage()
  n29 = NamedRerouteDeclaration(None=ComponentMask(None=@n28) {g=true}) {name=Width Scale, nodeColor=<rgba 0 0 0 1>}
  n30 = NamedRerouteUsage()
  n31 = Abs(None=Cosine(None=@n30))
  n32 = Abs(None=Sine(None=@n30))
  n33 = NamedRerouteDeclaration(None=ComponentMask(None=@n28) {r=true}) {name=Height Scale, nodeColor=<rgba 0 0 0 1>}
  NamedRerouteDeclaration(None=Add(A=Multiply(A=@n29, B=@n31), B=Multiply(A=@n32, B=@n33))) {name=Rotated Scale, nodeColor=<rgba 0 0 0 1>}
  NamedRerouteDeclaration(None=MaterialFunctionCall!(Dimensions=ViewSize()) {materialFunction=/FortUILibrary/MaterialFunctions/Basics/MF_UI_GetAspectRatio}.Max Ratio) {name=Max Ratio, nodeColor=<rgba 0 0 0 1>}
  NamedRerouteDeclaration(None=MaterialFunctionCall(Value=@n20.Custom Curve UV, Height=Constant() {r=1}) {materialFunction=&MF_UI_PingPong}.Result) {name=Curve, nodeColor=<rgba 0 0 0 1>}
  n34 = Floor(None=@n10)
  NamedRerouteDeclaration(None=MaterialFunctionCall!(UV=ComponentMask(None=@n34) {g=true}, Seed=Constant() {r=0.3}) {materialFunction=&MF_UI_Random}.Result) {name=Random Rows, nodeColor=<rgba 0 0 0 1>}
  n35 = MaterialFunctionCall(ScreenUV=@n8.ViewportUV) {materialFunction=&MF_UI_ScreenSpaceUVScaling}
  n36 = NamedRerouteUsage()
  n37 = Divide(A=Subtract(A=Rotator(Coordinate=@n35.Scaled to Fit, Time=@n36) {speed=6.2832}) {constB=0.5}, B=NamedRerouteUsage())
  NamedRerouteDeclaration(None=ComponentMask(None=Add(A=@n37) {constB=0.5}) {r=true}) {name=Horizontal Gradient, nodeColor=<rgba 0 0 0 1>}
 ]
}
