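# Editor utility widget for a dismissible info label in FortTechAudioTools' AI Voice Generator: an icon, text and Clear button over a colored border.
# PreConstruct calls ApplySettings, which pushes variables into text, tooltip, colors, icon and fixed text width.
# The FontSize setter has no incoming execution link, so that variable is not applied.
# Clicking Clear broadcasts OnClearClicked and removes the widget; the Clear event removes it without broadcasting.
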EditorUtilityWidgetBlueprint /FortTechAudioTools/Widgets/AIVoiceGenerator/EUW_ClearableInfoLabel
graphs [
 ApplySettings {nodes [
   ApplySettings() {is=title K2Node_FunctionEntry}
   > Set ColorAndOpacity!(
    execute
    ColorAndOpacity:Slate Color Structure=Make Slate Color(SpecifiedColor=Get TextColor() -> (TextColor:Linear Color Structure))
    self:Text Object Reference=Get TextBlock_Body() -> (TextBlock_Body:Text Object Reference)
   ) -> (then, Output_Get:Slate Color Structure)
   n1 = Get Text!() -> (Text)
   > "SetText (Text)"(self=Get TextBlock_Body(), InText=@n1)
   > n2 = SetToolTipText!(execute, self:Widget Object Reference, InToolTipText:"Text (by ref)"=@n1) -> (then)
   n3 = Get TechAudioToolsEditorSlateIcon() -> (TechAudioToolsEditorSlateIcon:Tech Audio Tools Editor Slate Icon Object Reference)
   SetIconSize!(
    execute=[@n2.then, @n8.then]
    self:Tech Audio Tools Editor Slate Icon Object Reference=@n3
    InSize:Vector 2D Structure=Get IconSize() -> (IconSize:Vector 2D Structure)
   ) -> (then)
   > SetIconBrushByPath!(execute, self:Tech Audio Tools Editor Slate Icon Object Reference=@n3, InFullIconPath:Name=Get FullIconPath() -> (FullIconPath:Name=None)) -> (then)
   > SetIconColor!(
    execute
    self:Tech Audio Tools Editor Slate Icon Object Reference=@n3
    InColor:Linear Color Structure=Get IconColor() -> (IconColor:Linear Color Structure)
   ) -> (then)
   > SetBrushColor(
    self=Get BackgroundBorder() -> (BackgroundBorder:Border Object Reference)
    InBrushColor=Get BackgroundColor() -> (BackgroundColor:Linear Color Structure)
   )
   n4 = Get SizeBox_Text() -> (SizeBox_Text:Size Box Object Reference)
   n5 = Get TextWidth() -> (TextWidth:float)
   > SetMinDesiredWidth!(execute, self:Size Box Object Reference=@n4, InMinDesiredWidth:float=@n5) -> (then)
   > SetMaxDesiredWidth!(execute, self:Size Box Object Reference=@n4, InMaxDesiredWidth:float=@n5) -> (then)
   n6 = Get TextBlock_Body()
   n7 = Get!(self:Text Object Reference=@n6) -> (
    Font_FontObject:Object Reference
    Font_FontMaterial:Object Reference
    Font_OutlineSettings:Font Outline Settings Structure
    Font_TypefaceFontName:Name
    Font_Size:float=~
    Font_LetterSpacing:int
    Font_SkewAmount:float=~
    Font_bForceMonospaced:bool=~
    Font_bMaterialIsStencil:bool=~
    Font_MonospacedWidth:float=~
   )
   n8 = Set!(
    execute
    Font_FontObject:Object Reference=@n7
    Font_FontMaterial:Object Reference=@n7.Font_FontMaterial
    Font_OutlineSettings:Font Outline Settings Structure=@n7.Font_OutlineSettings
    Font_TypefaceFontName:Name=@n7.Font_TypefaceFontName
    Font_Size:float=Get FontSize() -> (FontSize:double)
    Font_LetterSpacing:int=@n7.Font_LetterSpacing
    Font_SkewAmount:float=@n7.Font_SkewAmount
    Font_bForceMonospaced:bool=@n7.Font_bForceMonospaced
    Font_bMaterialIsStencil:bool=@n7.Font_bMaterialIsStencil
    Font_MonospacedWidth:float=@n7.Font_MonospacedWidth
    self:Text Object Reference=@n6
   ) -> (then, Output_Get:Slate Font Info Structure)
  ]}
 EventGraph {nodes [
   "On Clicked (EUW_Button_Secondary)"() {is=title K2Node_ComponentBoundEvent}
   > n1 = Call On Clear Clicked!(execute:Exec, self:EUW Clearable Info Label Object Reference) -> (then:Exec) {is=title K2Node_CallDelegate}
   RemoveFromParent(execute=[@n1.then, @n2.then])
   n2 = Clear() {is=title K2Node_CustomEvent}
   Event PreConstruct()
   > ApplySettings() {is=title K2Node_CallFunction}
  ]}
 OnClearClicked {nodes [OnClearClicked() {is=title K2Node_FunctionEntry}]}
]
variables [
 BackgroundColor {type=LinearColor, default=<rgba 1 0.7126 0.2678 1>}
 FontSize {type=number, default=12}
 FullIconPath {type=string, default=CoreStyle.Icons.Error.Large}
 IconColor {type=LinearColor, default=<rgba 1 0.2223 0 1>}
 IconSize {type=Vector2D, default=<16 16>}
 OnClearClicked {default=()}
 Text {type=string, default=Sample info message.}
 TextColor {type=LinearColor, default=<rgba 0 0 0 1>}
 TextWidth {type=number, default=500}
]
widgets {
 info {namedSlotCount=1, rootWidgetClass=&GridPanel, widgetCount=7}
 widgets [
  GridPanel_0 {widgetClassPath=&GridPanel, widget=GridPanel_0}
  BackgroundBorder {
   widgetClassPath=&Border
   bIsVariable=true
   parent=GridPanel_0
   widget=BackgroundBorder
   slot {class=GridSlot, properties {columnSpan=3, layer=-1}}
   properties=~
  }
  TechAudioToolsEditorSlateIcon {
   widgetClassPath=/Script/TechAudioToolsContentEditor.TechAudioToolsEditorSlateIcon
   bIsVariable=true
   parent=GridPanel_0
   widget=TechAudioToolsEditorSlateIcon
   properties {fullIconPath=CoreStyle.Icons.Error.Large}
   slot {class=GridSlot, properties {horizontalAlignment=HAlign_Center, verticalAlignment=VAlign_Center, padding=<ltrb 5 2 5 2>}}
  }
  SizeBox_Text {
   widgetClassPath=&SizeBox
   bIsVariable=true
   parent=GridPanel_0
   widget=SizeBox_Text
   properties {bOverride_MaxDesiredWidth=true, bOverride_MinDesiredWidth=true, maxDesiredWidth=500, minDesiredWidth=500}
   slot {class=GridSlot, properties {column=1, padding=<ltrb 0 4 0 2>}}
  }
  TextBlock_Body {
   widgetClassPath=&TextBlock
   bIsVariable=true
   parent=SizeBox_Text
   widget=TextBlock_Body
   properties {
    text=Sample info message
    font {
     bForceMonospaced=false
     bMaterialIsStencil=false
     fontMaterial=null
     fontObject=&Roboto
     letterSpacing=0
     monospacedWidth=1
     size=9
     skewAmount=0
     typefaceFontName=Bold
     outlineSettings {
      bApplyOutlineToDropShadows=false
      bMiteredCorners=false
      bSeparateFillAlpha=false
      outlineMaterial=null
      outlineSize=0
      outlineColor=<rgba 0 0 0 1>
     }
    }
   }
   slot {class=SizeBoxSlot}
  }
  EUW_Button_Secondary {
   widgetClassPath=/TechAudioToolsContent/Widgets/Editor/Buttons/EUW_Button_Secondary
   bIsVariable=true
   parent=GridPanel_0
   widget=EUW_Button_Secondary
   slot {class=GridSlot, properties {column=2, horizontalAlignment=HAlign_Center, verticalAlignment=VAlign_Center, padding=<ltrb 0 0 5 0>}}
  }
  TextBlock_62 {
   widgetClassPath=&TextBlock
   namedSlotHost=EUW_Button_Secondary
   widget=TextBlock_62
   properties {
    text=Clear
    font {
     bForceMonospaced=false
     bMaterialIsStencil=false
     fontMaterial=null
     fontObject=&Roboto
     letterSpacing=0
     monospacedWidth=1
     size=7.5
     skewAmount=0
     typefaceFontName=Bold
     outlineSettings {
      bApplyOutlineToDropShadows=false
      bMiteredCorners=false
      bSeparateFillAlpha=false
      outlineMaterial=null
      outlineSize=0
      outlineColor=<rgba 0 0 0 1>
     }
    }
   }
  }
 ]
}
