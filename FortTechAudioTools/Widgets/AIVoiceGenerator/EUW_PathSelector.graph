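# Reusable editor utility widget for the audio tools: a titled path field with browse and open-externally buttons.
# The folder button opens a file or directory dialog per FileSelectorType, starting at InitialSearchPath or the project directory.
# Picked or typed paths go through SetSelectedFilePath, which broadcasts changes; browse and open results flash green or red.
# A looping timer polls folder-button hover to broadcast changes and swap its Slate icon.
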EditorUtilityWidgetBlueprint /FortTechAudioTools/Widgets/AIVoiceGenerator/EUW_PathSelector
animations [
 Success {
  binding_count=2
  length_seconds=0.5
  object=&self:Success
  sequence {
   bindings [
    BG_Failure {
     id=#1
     parent_id=null
     possessed_class=Image
     tracks [{
       type=MovieSceneByteTrack
       display_name=Visibility
       sections [{type=MovieSceneByteSection, end_frame=0, start_frame=0, channels [None {type=MovieSceneScriptingByteChannel, const [2,1]}]}]
      }]
    }
    BG_Success {
     id=#2
     parent_id=null
     possessed_class=Image
     tracks [
      {
       type=MovieSceneByteTrack
       display_name=Visibility
       sections [{type=MovieSceneByteSection, end_frame=10, start_frame=0, channels [None {type=MovieSceneScriptingByteChannel, const [0,2]}]}]
      }
      {
       type=MovieSceneColorTrack
       display_name=Color and Opacity
       sections [{
         type=MovieSceneColorSection
         end_frame=10
         start_frame=0
         channels [
          Color.R {type=MovieSceneScriptingFloatChannel, const [0,1]}
          Color.G {type=MovieSceneScriptingFloatChannel, const [0.6719,1]}
          Color.B {type=MovieSceneScriptingFloatChannel, const [0.0345,1]}
          Color.A {type=MovieSceneScriptingFloatChannel, keys [[0,0.4],[0.25,1],[3,0.4],[10,0]]}
         ]
        }]
      }
     ]
    }
   ]
   display_rate {denominator=1, numerator=20}
   tick_resolution {denominator=1, numerator=60000}
   tracks []
  }
 }
 Failure {
  binding_count=3
  length_seconds=0.5
  object=&self:Failure
  sequence {
   bindings [
    BG_Success {
     id=#3
     parent_id=null
     possessed_class=Image
     tracks [{
       type=MovieSceneByteTrack
       display_name=Visibility
       sections [{type=MovieSceneByteSection, end_frame=10, start_frame=0, channels [None {type=MovieSceneScriptingByteChannel, const [2,2]}]}]
      }]
    }
    TextBlock_Title {
     id=#4
     parent_id=null
     possessed_class=TextBlock
     tracks [{
       type=MovieSceneFloatTrack
       display_name=Min Desired Width
       sections [{type=MovieSceneFloatSection, end_frame=0, start_frame=0, channels [None {type=MovieSceneScriptingFloatChannel, const [82.2431,1]}]}]
      }]
    }
    BG_Failure {
     id=#5
     parent_id=null
     possessed_class=Image
     tracks [
      {
       type=MovieSceneByteTrack
       display_name=Visibility
       sections [{type=MovieSceneByteSection, end_frame=10, start_frame=0, channels [None {type=MovieSceneScriptingByteChannel, const [0,2]}]}]
      }
      {
       type=MovieSceneColorTrack
       display_name=Color and Opacity
       sections [{
         type=MovieSceneColorSection
         end_frame=10
         start_frame=0
         channels [
          Color.R {type=MovieSceneScriptingFloatChannel, const [1,1]}
          Color.G {type=MovieSceneScriptingFloatChannel, const [0,1]}
          Color.B {type=MovieSceneScriptingFloatChannel, const [0.0843,1]}
          Color.A {type=MovieSceneScriptingFloatChannel, keys [[0,0],[0.25,1],[3,0.4],[10,0]]}
         ]
        }]
      }
     ]
    }
   ]
   display_rate {denominator=1, numerator=20}
   tick_resolution {denominator=1, numerator=60000}
   tracks []
  }
 }
]
graphs [
 EventGraph {nodes [
   "On Clicked (EUW_Button_Secondary_Icon)"() {is=title K2Node_ComponentBoundEvent}
   > n1 = GetEditorWorld(self=Get UnrealEditorSubsystem())
   n2 = Get EditableTextBox() -> (EditableTextBox:Editor Utility Editable Text Box Object Reference)
   > n3 = OpenFileInDefaultApplication!(execute, FilePath:String="To String (Text)"(InText=Get Text(self=@n2)), __WorldContext:Object Reference=@n1) -> (then, Success:bool)
   > n4 = Branch(Condition=@n3)
   PlayAnimation(
    execute=[@n5.then, @n4.then]
    InAnimation=Select(
     Option 0:Widget Animation Object Reference=Get Failure() -> (Failure:Widget Animation Object Reference)
     Option 1:Widget Animation Object Reference=Get Success() -> (Success:Widget Animation Object Reference)
     Index=@n3
    ) -> (ReturnValue:Widget Animation Object Reference)
   )
   n5 = Show Message Dialog(
    execute=@n4.else
    Title=Format Text(Format="LOCTEXT(\"Error\")")
    Message=Format Text(Format="LOCTEXT(\"Could not open in external application. Path does not exist.\")")
    MessageType=Ok
    DefaultValue=Ok
    MessageCategory=Warning
   )
   n6 = OnButtonHoverStateChanged_Event!() -> (OutputDelegate, then, IsHovered:bool=~) {is=title K2Node_CustomEvent}
   > SetSlateIcon!(
    execute
    self
    SlateIconName:Name=Get Slate Brush Name for Hover State!(IsHovered:bool=@n6.IsHovered) -> (ReturnValue:Name) {is=title K2Node_MacroInstance}
   ) -> (then)
   OnButtonClicked_Event() {is=title K2Node_CustomEvent}
   > n7 = Switch on E_FortTechAudioTools_FileSelectorType!(
    execute
    Selection:E_FortTechAudioTools_FileSelectorType Enum=Get FileSelectorType() -> (FileSelectorType:E_FortTechAudioTools_FileSelectorType Enum=NewEnumerator0)
   ) -> (NewEnumerator0:Exec, NewEnumerator1:Exec, NewEnumerator2:Exec) {is=title K2Node_SwitchEnum}
   n8 = OpenDirectorySelector!(
    execute=@n7.NewEnumerator1
    WindowTitle:String=Get WindowTitle() -> (WindowTitle:String)
    InitialSearchPath:String=Get Initial Search Directory!() -> (ReturnValue:String) {is=title K2Node_MacroInstance}
    __WorldContext:Object Reference
   ) -> (then, Success:bool, SelectedDirectory:String)
   > n9 = Branch(Condition=@n8)
   > n10 = SetSelectedFilePath!(execute, self, SelectedFilePath:String=@n8.SelectedDirectory) -> (then)
   PlayAnimation(execute=[@n10.then, @n9.else], InAnimation=Select(Option 0=Get Failure(), Option 1=Get Success(), Index=@n8))
   n11 = OpenFileSelector!(
    execute=@n7.NewEnumerator0
    WindowTitle:String=Get WindowTitle()
    InitialFileSearchPath:String=Get Initial Search Directory()
    FileTypeLabel:String=Get FileTypeLabel() -> (FileTypeLabel:String)
    FileTypeExtension:String=Get FileTypeExtension() -> (FileTypeExtension:String)
    __WorldContext:Object Reference
   ) -> (then, Success:bool, SelectedFilePath:String)
   > n12 = Branch(Condition=@n11)
   > n13 = SetSelectedFilePath(SelectedFilePath=@n11.SelectedFilePath)
   PlayAnimation(execute=[@n13.then, @n12.else], InAnimation=Select(Option 0=Get Failure(), Option 1=Get Success(), Index=@n11))
   n14 = TextChanged_Event!() -> (OutputDelegate, then, Text:"Text (by ref)") {is=title K2Node_CustomEvent}
   > SetSelectedFilePath(SelectedFilePath="To String (Text)"(InText=@n14.Text))
   Event PreConstruct()
   > ResetHighlights()
   > UpdateEditableTextBoxWidth()
   > Set HintText!(execute, HintText:Text=Get TextBoxHint() -> (TextBoxHint:Text), self:Text Box Object Reference=Get EditableTextBox()) -> (then, Output_Get:Text)
   > Set Text(Text=Get Title() -> (Title:Text), self=Get TextBlock_Title() -> (TextBlock_Title:Text Object Reference))
   > SetSlateIcon(SlateIconName=Get Default Slate Icon Override() -> (Default Slate Icon Override:Name=None))
   Event Construct()
   > Bind Event to On Button Hover State Changed(self:Fort Tech Audio Tools Path Selector Object Reference, Delegate=Create Event()) {is=title K2Node_AddDelegate}
   > Bind Event to On Clicked(
    self:Tech Audio Tools Button Secondary Icon Object Reference=Get EUW_Button_Secondary_Icon_Folder() -> (EUW_Button_Secondary_Icon_Folder:Tech Audio Tools Button Secondary Icon Object Reference)
    Delegate=Create Event()
   ) {is=title K2Node_AddDelegate}
   > Bind Event to On Text Changed(self:Text Box Object Reference=Get EditableTextBox(), Delegate=Create Event()) {is=title K2Node_AddDelegate}
   > n15 = Set Timer by Function Name!(
    execute
    Object:Object Reference
    FunctionName:String=UpdateWidgetIconState
    Time:float=0.05
    bLooping:bool=true
    bMaxOncePerFrame:bool=true
    InitialStartDelay:float=0
    InitialStartDelayVariance:float=0
   ) -> (then, ReturnValue:Timer Handle Structure) {is=title K2Node_CallFunction}
   > Set UpdateTimer(UpdateTimer:Timer Handle Structure=@n15) -> (Output_Get:Timer Handle Structure)
   Event Destruct()
   > Clear and Invalidate Timer by Handle!(execute, WorldContextObject:Object Reference, Handle=Get UpdateTimer() -> (UpdateTimer:Timer Handle Structure)) -> (then)
   > Unbind all Events from On Button Hover State Changed!(execute:Exec, self:Fort Tech Audio Tools Path Selector Object Reference) -> (then:Exec) {is=title K2Node_ClearDelegate}
   > Unbind all Events from On Clicked!(execute:Exec, self:Tech Audio Tools Button Secondary Icon Object Reference=Get EUW_Button_Secondary_Icon_Folder()) -> (then:Exec) {is=title K2Node_ClearDelegate}
  ]}
 GetInitialSearchDirectory {nodes [
   Inputs!()
   n1 = Get InitialSearchPath() -> (InitialSearchPath:String)
   Outputs!(ReturnValue:String=Select(
     Option 0:String=GetProjectDirectory!() -> (ReturnValue:String)
     Option 1:String=@n1
     Index=ValidatePath!(InPath:String=@n1) -> (bDidSucceed:bool, OutReason:Text)
    ) -> (ReturnValue:String))
  ]}
 GetSelectedFilePath {nodes [
   GetSelectedFilePath() {is=title K2Node_FunctionEntry}
   > Return Node!(execute, SelectedFilePath:String=Get SelectedFilePath() -> (SelectedFilePath:String))
  ]}
 GetSlateBrushNameForHoverState {nodes [
   n1 = Get EUW_Button_Secondary_Icon_Folder()
   Outputs(ReturnValue:Name=Select(
     Option 0:Name=Get Icon Brush Path!(self:Tech Audio Tools Button Secondary Icon Object Reference=@n1) -> (Icon Brush Path:Name=None)
     Option 1:Name=Get Hovered Slate Icon Override() -> (Hovered Slate Icon Override:Name=None)
     Index=Inputs!() -> (IsHovered:bool=~)
    ) -> (ReturnValue:Name))
  ]}
 OnButtonHoverStateChanged {nodes [OnButtonHoverStateChanged!() -> (then, IsHovered:bool=~) {is=title K2Node_FunctionEntry}]}
 OnSelectedFilePathChanged {nodes [OnSelectedFilePathChanged!() -> (then, SelectedFilePath:String) {is=title K2Node_FunctionEntry}]}
 ResetHighlights {nodes [
   ResetHighlights() {is=title K2Node_FunctionEntry}
   > SetVisibility(self=[Get BG_Success() -> (BG_Success:Image Object Reference), Get BG_Failure() -> (BG_Failure:Image Object Reference)], InVisibility=Hidden)
  ]}
 SetInitialSearchPath {nodes [
   n1 = SetInitialSearchPath!() -> (then, InitialSearchPath:String) {is=title K2Node_FunctionEntry}
   > Set InitialSearchPath(InitialSearchPath:String=@n1) -> (Output_Get:String)
  ]}
 SetSelectedFilePath {nodes [
   n1 = SetSelectedFilePath!() -> (then, SelectedFilePath:String) {is=title K2Node_FunctionEntry}
   > Branch(Condition="Not Equal Exactly (String)"(A=Get SelectedFilePath(), B=@n1))
   > n2 = Set SelectedFilePath(SelectedFilePath:String=@n1) -> (Output_Get:String)
   > Call On Selected File Path Changed!(execute:Exec, self:Fort Tech Audio Tools Path Selector Object Reference, SelectedFilePath:String=@n2) -> (then:Exec) {is=title K2Node_CallDelegate}
   > Set Text(Text="To Text (String)"(InString=@n2), self:Text Box Object Reference=Get EditableTextBox())
  ]}
 SetSlateIcon {nodes [
   n1 = SetSlateIcon!() -> (then, SlateIconName:Name) {is=title K2Node_FunctionEntry}
   > Branch(Condition="Not Equal (Name)"(A=@n1) -> (ReturnValue=~))
   > SetIconBrushByPath!(
    execute
    self:Tech Audio Tools Editor Slate Icon Object Reference=Get Icon!(self:Tech Audio Tools Button Secondary Icon Object Reference=Get EUW_Button_Secondary_Icon_Folder()) -> (Icon:Tech Audio Tools Editor Slate Icon Object Reference)
    InFullIconPath:Name=@n1
   ) -> (then)
  ]}
 UpdateEditableTextBoxWidth {nodes [
   UpdateEditableTextBoxWidth() {is=title K2Node_FunctionEntry}
   n1 = Get SizeBox_EditableTextBox() -> (SizeBox_EditableTextBox:Size Box Object Reference)
   n2 = Get TextBoxWidth() -> (TextBoxWidth:double)
   > SetWidthOverride(self=@n1, InWidthOverride=@n2)
   > SetMinDesiredWidth!(execute, self:Size Box Object Reference=@n1, InMinDesiredWidth:float=@n2) -> (then)
   > SetMaxDesiredWidth!(execute, self:Size Box Object Reference=@n1, InMaxDesiredWidth:float=@n2) -> (then)
  ]}
 UpdateWidgetIconState {nodes [
   UpdateWidgetIconState() {is=title K2Node_FunctionEntry}
   n1 = IsHovered(self=Get EUW_Button_Secondary_Icon_Folder())
   > Branch(Condition="Not Equal (Boolean)"(A=Get IsHoveredInternal() -> (IsHoveredInternal:bool), B=@n1) -> (ReturnValue=~))
   > n2 = Set IsHoveredInternal(IsHoveredInternal:bool=@n1) -> (Output_Get:bool)
   > Call On Button Hover State Changed!(execute:Exec, self:Fort Tech Audio Tools Path Selector Object Reference, IsHovered:bool=@n2) -> (then:Exec) {is=title K2Node_CallDelegate}
  ]}
]
variables [
 Default Slate Icon Override {type=string, default=EditorStyle.ContentBrowser.AssetTreeFolderClosed}
 DefaultPath {type=string, default=""}
 FileSelectorType {type=E_FortTechAudioTools_FileSelectorType, default=File}
 FileTypeExtension {type=string, default=.csv}
 FileTypeLabel {type=string, default=CSV File}
 Hovered Slate Icon Override {type=string, default=EditorStyle.ContentBrowser.AssetTreeFolderOpen}
 InitialSearchPath {type=string, default=""}
 IsHoveredInternal {type=boolean}
 OnButtonHoverStateChanged {default=()}
 OnSelectedFilePathChanged {default=()}
 SelectedFilePath {type=string, default=""}
 TextBoxHint {type=string, default=Enter text here}
 TextBoxWidth {type=number, default=200}
 Title {type=string, default=Directory}
 UpdateTimer {type=TimerHandle}
 WindowTitle {type=string, default=Open File}
]
widgets {
 info {rootWidgetClass=&GridPanel, widgetCount=11}
 widgets [
  GridPanel_0 {widgetClassPath=&GridPanel, widget=GridPanel_0}
  TextBlock_Title {
   widgetClassPath=&TextBlock
   bIsVariable=true
   parent=GridPanel_0
   widget=TextBlock_Title
   properties {
    minDesiredWidth=82.2431
    font {
     bForceMonospaced=false
     bMaterialIsStencil=false
     fontMaterial=null
     fontObject=/LauncherSocial/BrutalType
     letterSpacing=0
     monospacedWidth=1
     size=10.5
     skewAmount=0
     typefaceFontName=Light
     outlineSettings {
      bApplyOutlineToDropShadows=false
      bMiteredCorners=false
      bSeparateFillAlpha=false
      outlineMaterial=null
      outlineSize=0
      outlineColor=<rgba 0 0 0 1>
     }
    }
   }
   slot {class=GridSlot, properties {padding=<ltrb 2 0 0 0>}}
  }
  GridPanel_26 {widgetClassPath=&GridPanel, parent=GridPanel_0, widget=GridPanel_26, slot {class=GridSlot, properties {row=1}}}
  EUW_Button_Secondary_Icon_Folder {
   widgetClassPath=/TechAudioToolsContent/Widgets/Editor/Buttons/EUW_Button_Secondary_Icon
   bIsVariable=true
   parent=GridPanel_26
   widget=EUW_Button_Secondary_Icon_Folder
   properties {icon Brush Path=EditorStyle.ContentBrowser.AssetTreeFolderClosed, renderTransform {angle=0, scale=<0.9 1>, shear=<0 0>, translation=<0 0>}}
   slot {class=GridSlot, properties {row=1}}
  }
  SizeBox_EditableTextBox {
   widgetClassPath=&SizeBox
   bIsVariable=true
   parent=GridPanel_26
   widget=SizeBox_EditableTextBox
   slot {class=GridSlot, properties {column=1, row=1}}
  }
  EditableTextBox {
   widgetClassPath=&EditorUtilityEditableTextBox
   bIsVariable=true
   parent=SizeBox_EditableTextBox
   widget=EditableTextBox
   properties {
    hintText=Enter text here ewqfgajlkcgheajlcvyha
    widgetStyle {
     backgroundColor {colorUseRule=UseColor_Specified, specifiedColor=<rgba 1 1 1 1>}
     backgroundImageFocused {
      bIsDynamicallyLoaded=false
      drawAs=RoundedBox
      imageType=NoImage
      mirroring=NoMirror
      resourceName=None
      resourceObject=null
      tiling=NoTile
      imageSize=<0 0>
      margin=<ltrb 0 0 0 0>
      outlineSettings {
       bUseBrushTransparency=false
       roundingType=FixedRadius
       width=1
       color {colorUseRule=UseColor_ColorTable, specifiedColor=<rgba 0 0 0 0>}
       cornerRadii=<4 4 4 4>
      }
      tintColor {colorUseRule=UseColor_Specified, specifiedColor=<rgba 0.0048 0.0048 0.0048 1>}
      uVRegion {bIsValid=false, max=<0 0>, min=<0 0>}
     }
     backgroundImageHovered {
      bIsDynamicallyLoaded=false
      drawAs=RoundedBox
      imageType=NoImage
      mirroring=NoMirror
      resourceName=None
      resourceObject=null
      tiling=NoTile
      imageSize=<0 0>
      margin=<ltrb 0 0 0 0>
      outlineSettings {
       bUseBrushTransparency=false
       roundingType=FixedRadius
       width=1
       color {colorUseRule=UseColor_ColorTable, specifiedColor=<rgba 0 0 0 0>}
       cornerRadii=<4 4 4 4>
      }
      tintColor {colorUseRule=UseColor_Specified, specifiedColor=<rgba 0.0048 0.0048 0.0048 1>}
      uVRegion {bIsValid=false, max=<0 0>, min=<0 0>}
     }
     backgroundImageNormal {
      bIsDynamicallyLoaded=false
      drawAs=RoundedBox
      imageType=NoImage
      mirroring=NoMirror
      resourceName=None
      resourceObject=null
      tiling=NoTile
      imageSize=<0 0>
      margin=<ltrb 0 0 0 0>
      outlineSettings {
       bUseBrushTransparency=false
       roundingType=FixedRadius
       width=1
       color {colorUseRule=UseColor_ColorTable, specifiedColor=<rgba 0 0 0 0>}
       cornerRadii=<4 4 4 4>
      }
      tintColor {colorUseRule=UseColor_Specified, specifiedColor=<rgba 0.0111 0.0299 0.125 1>}
      uVRegion {bIsValid=false, max=<0 0>, min=<0 0>}
     }
     backgroundImageReadOnly {
      bIsDynamicallyLoaded=false
      drawAs=RoundedBox
      imageType=NoImage
      mirroring=NoMirror
      resourceName=None
      resourceObject=null
      tiling=NoTile
      imageSize=<0 0>
      margin=<ltrb 0 0 0 0>
      outlineSettings {
       bUseBrushTransparency=false
       roundingType=FixedRadius
       width=1
       color {colorUseRule=UseColor_ColorTable, specifiedColor=<rgba 0 0 0 0>}
       cornerRadii=<4 4 4 4>
      }
      tintColor {colorUseRule=UseColor_Specified, specifiedColor=<rgba 0.0284 0.0284 0.0284 1>}
      uVRegion {bIsValid=false, max=<0 0>, min=<0 0>}
     }
     focusedForegroundColor {colorUseRule=UseColor_Specified, specifiedColor=<rgba 1 1 1 1>}
     foregroundColor {colorUseRule=UseColor_Specified, specifiedColor=<rgba 0.5271 0.5271 0.5271 1>}
     hScrollBarPadding=<ltrb 0 0 0 0>
     padding=<ltrb 8 3 8 4>
     readOnlyForegroundColor {colorUseRule=UseColor_Foreground, specifiedColor=<rgba 1 0 1 1>}
     scrollBarStyle {
      thickness=8
      draggedThumbImage {
       bIsDynamicallyLoaded=false
       drawAs=RoundedBox
       imageType=NoImage
       mirroring=NoMirror
       resourceName=None
       resourceObject=null
       tiling=NoTile
       imageSize=<0 0>
       margin=<ltrb 0 0 0 0>
       outlineSettings {
        bUseBrushTransparency=false
        roundingType=FixedRadius
        width=0
        color {colorUseRule=UseColor_Specified, specifiedColor=<rgba 0 0 0 0>}
        cornerRadii=<4 4 4 4>
       }
       tintColor {colorUseRule=UseColor_Specified, specifiedColor=<rgba 0.2159 0.2159 0.2159 1>}
       uVRegion {bIsValid=false, max=<0 0>, min=<0 0>}
      }
      horizontalBackgroundImage {
       bIsDynamicallyLoaded=false
       drawAs=NoDrawType
       imageType=NoImage
       mirroring=NoMirror
       resourceName=None
       resourceObject=null
       tiling=NoTile
       imageSize=<0 0>
       margin=<ltrb 0 0 0 0>
       outlineSettings {
        bUseBrushTransparency=false
        roundingType=HalfHeightRadius
        width=0
        color {colorUseRule=UseColor_Specified, specifiedColor=<rgba 0 0 0 0>}
        cornerRadii=<0 0 0 0>
       }
       tintColor {colorUseRule=UseColor_Specified, specifiedColor=<rgba 1 1 1 1>}
       uVRegion {bIsValid=false, max=<0 0>, min=<0 0>}
      }
      horizontalBottomSlotImage {
       bIsDynamicallyLoaded=false
       drawAs=NoDrawType
       imageType=NoImage
       mirroring=NoMirror
       resourceName=None
       resourceObject=null
       tiling=NoTile
       imageSize=<0 0>
       margin=<ltrb 0 0 0 0>
       outlineSettings {
        bUseBrushTransparency=false
        roundingType=HalfHeightRadius
        width=0
        color {colorUseRule=UseColor_Specified, specifiedColor=<rgba 0 0 0 0>}
        cornerRadii=<0 0 0 0>
       }
       tintColor {colorUseRule=UseColor_Specified, specifiedColor=<rgba 1 1 1 1>}
       uVRegion {bIsValid=false, max=<0 0>, min=<0 0>}
      }
      horizontalTopSlotImage {
       bIsDynamicallyLoaded=false
       drawAs=NoDrawType
       imageType=NoImage
       mirroring=NoMirror
       resourceName=None
       resourceObject=null
       tiling=NoTile
       imageSize=<0 0>
       margin=<ltrb 0 0 0 0>
       outlineSettings {
        bUseBrushTransparency=false
        roundingType=HalfHeightRadius
        width=0
        color {colorUseRule=UseColor_Specified, specifiedColor=<rgba 0 0 0 0>}
        cornerRadii=<0 0 0 0>
       }
       tintColor {colorUseRule=UseColor_Specified, specifiedColor=<rgba 1 1 1 1>}
       uVRegion {bIsValid=false, max=<0 0>, min=<0 0>}
      }
      hoveredThumbImage {
       bIsDynamicallyLoaded=false
       drawAs=RoundedBox
       imageType=NoImage
       mirroring=NoMirror
       resourceName=None
       resourceObject=null
       tiling=NoTile
       imageSize=<0 0>
       margin=<ltrb 0 0 0 0>
       outlineSettings {
        bUseBrushTransparency=false
        roundingType=FixedRadius
        width=0
        color {colorUseRule=UseColor_Specified, specifiedColor=<rgba 0 0 0 0>}
        cornerRadii=<4 4 4 4>
       }
       tintColor {colorUseRule=UseColor_Specified, specifiedColor=<rgba 0.2159 0.2159 0.2159 1>}
       uVRegion {bIsValid=false, max=<0 0>, min=<0 0>}
      }
      normalThumbImage {
       bIsDynamicallyLoaded=false
       drawAs=RoundedBox
       imageType=NoImage
       mirroring=NoMirror
       resourceName=None
       resourceObject=null
       tiling=NoTile
       imageSize=<0 0>
       margin=<ltrb 0 0 0 0>
       outlineSettings {
        bUseBrushTransparency=false
        roundingType=FixedRadius
        width=0
        color {colorUseRule=UseColor_Specified, specifiedColor=<rgba 0 0 0 0>}
        cornerRadii=<4 4 4 4>
       }
       tintColor {colorUseRule=UseColor_Specified, specifiedColor=<rgba 0.0953 0.0953 0.0953 1>}
       uVRegion {bIsValid=false, max=<0 0>, min=<0 0>}
      }
      verticalBackgroundImage {
       bIsDynamicallyLoaded=false
       drawAs=NoDrawType
       imageType=NoImage
       mirroring=NoMirror
       resourceName=None
       resourceObject=null
       tiling=NoTile
       imageSize=<0 0>
       margin=<ltrb 0 0 0 0>
       outlineSettings {
        bUseBrushTransparency=false
        roundingType=HalfHeightRadius
        width=0
        color {colorUseRule=UseColor_Specified, specifiedColor=<rgba 0 0 0 0>}
        cornerRadii=<0 0 0 0>
       }
       tintColor {colorUseRule=UseColor_Specified, specifiedColor=<rgba 1 1 1 1>}
       uVRegion {bIsValid=false, max=<0 0>, min=<0 0>}
      }
      verticalBottomSlotImage {
       bIsDynamicallyLoaded=false
       drawAs=NoDrawType
       imageType=NoImage
       mirroring=NoMirror
       resourceName=None
       resourceObject=null
       tiling=NoTile
       imageSize=<0 0>
       margin=<ltrb 0 0 0 0>
       outlineSettings {
        bUseBrushTransparency=false
        roundingType=HalfHeightRadius
        width=0
        color {colorUseRule=UseColor_Specified, specifiedColor=<rgba 0 0 0 0>}
        cornerRadii=<0 0 0 0>
       }
       tintColor {colorUseRule=UseColor_Specified, specifiedColor=<rgba 1 1 1 1>}
       uVRegion {bIsValid=false, max=<0 0>, min=<0 0>}
      }
      verticalTopSlotImage {
       bIsDynamicallyLoaded=false
       drawAs=NoDrawType
       imageType=NoImage
       mirroring=NoMirror
       resourceName=None
       resourceObject=null
       tiling=NoTile
       imageSize=<0 0>
       margin=<ltrb 0 0 0 0>
       outlineSettings {
        bUseBrushTransparency=false
        roundingType=HalfHeightRadius
        width=0
        color {colorUseRule=UseColor_Specified, specifiedColor=<rgba 0 0 0 0>}
        cornerRadii=<0 0 0 0>
       }
       tintColor {colorUseRule=UseColor_Specified, specifiedColor=<rgba 1 1 1 1>}
       uVRegion {bIsValid=false, max=<0 0>, min=<0 0>}
      }
     }
     textStyle {
      overflowPolicy=Clip
      transformPolicy=None
      colorAndOpacity {colorUseRule=UseColor_Foreground, specifiedColor=<rgba 1 0 1 1>}
      font {
       bForceMonospaced=false
       bMaterialIsStencil=false
       fontMaterial=null
       fontObject=null
       letterSpacing=0
       monospacedWidth=1
       size=10
       skewAmount=0
       typefaceFontName=Regular
       outlineSettings {
        bApplyOutlineToDropShadows=false
        bMiteredCorners=false
        bSeparateFillAlpha=false
        outlineMaterial=null
        outlineSize=0
        outlineColor=<rgba 0 0 0 1>
       }
      }
      highlightColor {colorUseRule=UseColor_ColorTable, specifiedColor=<rgba 0 0 0 0>}
      highlightShape {
       bIsDynamicallyLoaded=false
       drawAs=Image
       imageType=NoImage
       mirroring=NoMirror
       resourceName=None
       resourceObject=null
       tiling=NoTile
       imageSize=<0 0>
       margin=<ltrb 0 0 0 0>
       outlineSettings {
        bUseBrushTransparency=false
        roundingType=HalfHeightRadius
        width=0
        color {colorUseRule=UseColor_Specified, specifiedColor=<rgba 0 0 0 0>}
        cornerRadii=<0 0 0 0>
       }
       tintColor {colorUseRule=UseColor_ColorTable, specifiedColor=<rgba 0 0 0 0>}
       uVRegion {bIsValid=false, max=<0 0>, min=<0 0>}
      }
      selectedBackgroundColor {colorUseRule=UseColor_ColorTable, specifiedColor=<rgba 0 0 0 0>}
      shadowColorAndOpacity=<rgba 0 0 0 1>
      shadowOffset=<0 0>
      strikeBrush {
       bIsDynamicallyLoaded=false
       drawAs=Image
       imageType=NoImage
       mirroring=NoMirror
       resourceName=None
       resourceObject=null
       tiling=NoTile
       imageSize=<32 32>
       margin=<ltrb 0 0 0 0>
       outlineSettings {
        bUseBrushTransparency=false
        roundingType=HalfHeightRadius
        width=0
        color {colorUseRule=UseColor_Specified, specifiedColor=<rgba 0 0 0 0>}
        cornerRadii=<0 0 0 0>
       }
       tintColor {colorUseRule=UseColor_Specified, specifiedColor=<rgba 1 1 1 1>}
       uVRegion {bIsValid=false, max=<0 0>, min=<0 0>}
      }
      underlineBrush {
       bIsDynamicallyLoaded=false
       drawAs=Image
       imageType=NoImage
       mirroring=NoMirror
       resourceName=None
       resourceObject=null
       tiling=NoTile
       imageSize=<32 32>
       margin=<ltrb 0 0 0 0>
       outlineSettings {
        bUseBrushTransparency=false
        roundingType=HalfHeightRadius
        width=0
        color {colorUseRule=UseColor_Specified, specifiedColor=<rgba 0 0 0 0>}
        cornerRadii=<0 0 0 0>
       }
       tintColor {colorUseRule=UseColor_Specified, specifiedColor=<rgba 1 1 1 1>}
       uVRegion {bIsValid=false, max=<0 0>, min=<0 0>}
      }
     }
     vScrollBarPadding=<ltrb 0 0 0 0>
    }
   }
   slot {class=SizeBoxSlot}
  }
  SizeBox_1 {
   widgetClassPath=&SizeBox
   parent=GridPanel_26
   widget=SizeBox_1
   properties {
    bOverride_MaxDesiredWidth=true
    clipping=ClipToBoundsAlways
    maxDesiredWidth=36.9957
    renderTransform {angle=0, scale=<0.75 0.7>, shear=<0 0>, translation=<0 0>}
   }
   slot {class=GridSlot, properties {column=2, horizontalAlignment=HAlign_Center, layer=5, row=1}}
  }
  EUW_Button_Secondary_Icon {
   widgetClassPath=&EUW_Button_Secondary_Icon
   bIsVariable=true
   parent=SizeBox_1
   widget=EUW_Button_Secondary_Icon
   properties {icon Brush Path=EditorStyle.MainFrame.Exit, icon Color=<rgba 0.6354 0.6354 0.6354 1>, icon Size=<37.1947 20>}
   slot {class=SizeBoxSlot, properties {horizontalAlignment=HAlign_Center, verticalAlignment=VAlign_Center}}
  }
  BG_Success {
   widgetClassPath=&Image
   bIsVariable=true
   parent=GridPanel_0
   widget=BG_Success
   properties {colorAndOpacity=<rgba 0 0.6719 0.0345 1>}
   slot {class=GridSlot, properties {rowSpan=2}}
  }
  BG_Failure {
   widgetClassPath=&Image
   bIsVariable=true
   parent=GridPanel_0
   widget=BG_Failure
   properties {colorAndOpacity=<rgba 1 0 0.0843 1>}
   slot {class=GridSlot, properties {rowSpan=2}}
  }
  BG_Main {
   widgetClassPath=&Image
   bIsVariable=true
   parent=GridPanel_0
   widget=BG_Main
   properties {colorAndOpacity=<rgba 0.0176 0.0176 0.0176 1>}
   slot {class=GridSlot, properties {layer=-1, rowSpan=2}}
  }
 ]
}
