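# Translucent UI material for the V-Bucks store screen background, built from layered blue gradients.
# A rotated linear gradient blends two dark blue parameters into the base, and a scaled, positionable exponential radial gradient supplies a brighter blue glow.
# An alpha composite branch factors in vertex alpha.
# A slowly panning flow-warp texture yields a small noise offset under BoltAnimation parameters, though no bolt sampling or final output wiring appears.
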Material /FortniteCoreUI/Frontend/StoreMain/Materials/M_UI_Screen_Vbucks_Background
graph {
 properties {
  bAutomaticallySetUsageInEditor=false
  bUsedWithStaticMesh=false
  blendMode=BLEND_Translucent
  materialDomain=MD_UI
  usedShadingModels=Default Lit
 }
 nodes [
  LinearInterpolate(B=NamedRerouteUsage())
  LinearInterpolate(A=NamedRerouteUsage(), Alpha=NamedRerouteUsage())
  MaterialFunctionCall() {materialFunction=&MF_UI_RadialGradient}
  n1 = NamedRerouteUsage()
  n2 = ScalarParameter() {defaultValue=-0.3, group=Gradient Base, parameterName=GradientBaseRotation}
  n3 = ScalarParameter() {defaultValue=0.3, group=Gradient Base, parameterName=GradientBaseIntensity}
  NamedRerouteDeclaration(None=LinearInterpolate(
    A=VectorParameter() {group=Gradient Base, parameterName=GradientBaseColor1, defaultValue=<rgba 0.001 0.0032 0.0521 1>}.RGB
    B=VectorParameter() {group=Gradient Base, parameterName=GradientBaseColor2, defaultValue=<rgba 0.0003 0.0307 0.1946 1>}.RGB
    Alpha=MaterialFunctionCall(UVs=@n1, RotationAngle=@n2, GradientOffset=@n3) {materialFunction=&MF_UI_LinearGradient}.Result
   )) {name=GradientBase, nodeColor=<rgba 1 0 0.2109 1>}
  n4 = MaterialFunctionCall(
   UVs=MaterialFunctionCall() {materialFunction=&MF_UI_UVScaling}.Scaled to Fit
   Texture Scale=AppendVector(
    A=ScalarParameter() {defaultValue=1.6131, group=Gradient Glow, parameterName=GradientGlowSizeX}
    B=ScalarParameter() {defaultValue=0.8823, group=Gradient Glow, parameterName=GradientGlowSizeY}
   )
  ) {materialFunction=&ScaleUVsByCenter}
  n5 = MaterialFunctionCall(
   X=ScalarParameter() {defaultValue=0.588, group=Gradient Glow, parameterName=GradientGlowPositionX}
   Y=ScalarParameter() {defaultValue=0.544, group=Gradient Glow, parameterName=GradientGlowPositionY}
  ) {materialFunction=&MakeFloat2}
  n6 = ScalarParameter() {defaultValue=0.792, group=Gradient Glow, parameterName=GradientGlowRadius}
  n7 = ScalarParameter() {defaultValue=1.0829, group=Gradient Glow, parameterName=GradientGlowDensity}
  NamedRerouteDeclaration(None=LinearInterpolate(
    B=VectorParameter() {group=Gradient Glow, parameterName=GradientGlowColor, defaultValue=<rgba 0 0.1291 0.5573 1>}.RGB
    Alpha=MaterialFunctionCall(UVs=@n4.UVs, CenterPosition=@n5.Result, Radius=@n6, Density=@n7) {materialFunction=&RadialGradientExponential}.RadialGradientExponential
   )) {name=Gradient Glow, nodeColor=<rgba 0.2188 1 0 1>}
  n8 = VertexColor()
  n9 = NamedRerouteUsage()
  n10 = Multiply(A=Constant3Vector() {constant=<rgba 1 1 1 1>}, B=@n9)
  n11 = Constant() {r=0.9}
  NamedRerouteDeclaration(None=LinearInterpolate(A=Multiply(A=@n8.A, B=@n10), Alpha=Multiply(A=@n9, B=@n11))) {name=GradientGlowAlphaComposite, nodeColor=<rgba 0 1 0.6484 1>}
  NamedRerouteDeclaration(None=MaterialFunctionCall(UV=TextureCoordinate()) {materialFunction=&MF_UI_UVScaling}.Scaled to Fit) {name=UV, nodeColor=<rgba 1 0.2578 0 1>}
  n12 = Panner() {speedX=0.008, speedY=0.008}
  NamedRerouteDeclaration(None=Multiply(
    A=TextureSample(UVs=@n12) {samplerType=SAMPLERTYPE_Normal, texture=/Game/Global/Textures/Default/Noise/T_FlowWarp_Noise}.R
    B=ScalarParameter() {defaultValue=0.005, group=BoltAnimation, parameterName=BoltDistortion, sliderMax=2}
   )) {name=Noise, nodeColor=<rgba 0 0.2734 1 1>}
 ]
}
