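# Editor Utility Widget for a text-labeled toggle button in TechAudioTools, built by placing a TextBlock in the named slot of EUW_ToggleButton_ContentSlot.
# PreConstruct pushes the exposed label, checked state and checkbox style to the inner button and text.
# Text color switches to the highlight color while checked or hovered, refreshing on state change and mouse enter/leave.
# The tooltip swaps between checked and unchecked text as the state changes.
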EditorUtilityWidgetBlueprint /TechAudioToolsContent/Widgets/Editor/ToggleButtons/EUW_ToggleButton_Text
graphs [EventGraph {nodes [
   "On Checked State Changed (ToggleButton_Content)"!() -> (OutputDelegate, then, Is Checked:bool) {is=title K2Node_ComponentBoundEvent}
   > Update Text Color()
   > Update Tool Tip()
   Update Text Color() {is=title K2Node_CustomEvent}
   n1 = IsHovered()
   n2 = Is Checked!(self:Tech Audio Tools Toggle Button Content Slot Object Reference=Get ToggleButton_Content() -> (ToggleButton_Content:Tech Audio Tools Toggle Button Content Slot Object Reference)) -> (Return Value:bool)
   > SetColorAndOpacity(
    self=Get TextBlock() -> (TextBlock:Text Object Reference)
    InColorAndOpacity=Select(
     Option 0:Slate Color Structure=Get Default Text Color() -> (Default Text Color:Slate Color Structure)
     Option 1:Slate Color Structure=Get Highlighted Text Color() -> (Highlighted Text Color:Slate Color Structure)
     Index=OR Boolean(A=@n2, B=@n1)
    ) -> (ReturnValue:Slate Color Structure)
   )
   Update Tool Tip() {is=title K2Node_CustomEvent}
   n3 = Get ToggleButton_Content()
   > SetToolTipText!(
    execute
    self:Widget Object Reference
    InToolTipText:"Text (by ref)"=SelectText(A=Get Checked Tool Tip() -> (Checked Tool Tip:Text), B=Get Unchecked Tool Tip() -> (Unchecked Tool Tip:Text), bPickA=Is Checked(self=@n3))
   ) -> (then)
   Event PreConstruct()
   > SetCheckedState!(
    execute
    self:Check Box Object Reference=Get ToggleButton!(self:Tech Audio Tools Toggle Button Content Slot Object Reference=Get ToggleButton_Content()) -> (ToggleButton:Check Box Object Reference)
    InCheckedState:ECheckBoxState Enum=Get Checked State() -> (Checked State:ECheckBoxState Enum=Unchecked)
   ) -> (then)
   > Update Text Color() {is=title K2Node_CallFunction}
   > Update Tool Tip() {is=title K2Node_CallFunction}
   > "SetText (Text)"(self=Get TextBlock(), InText=Get Text!() -> (Text))
   > Set WidgetStyle!(
    execute
    WidgetStyle:Check Box Style Structure=Get Check Box Style() -> (Check Box Style:Check Box Style Structure)
    self:Check Box Object Reference=Get ToggleButton(self=Get ToggleButton_Content())
   ) -> (then, Output_Get:Check Box Style Structure)
   n4 = Event OnMouseEnter()
   Update Text Color(execute=[@n4.then, @n5.then])
   n5 = Event OnMouseLeave!() -> (OutputDelegate, then, MouseEvent:"Pointer Event Structure (by ref)") {is=title K2Node_Event}
  ]}]
variables [
 Check Box Style {
  type=CheckBoxStyle
  default {
   checkBoxType=ToggleButton
   backgroundHoveredImage {
    bIsDynamicallyLoaded=false
    drawAs=RoundedBox
    imageType=NoImage
    mirroring=NoMirror
    resourceName=None
    resourceObject=null
    tiling=NoTile
    imageSize=<18 18>
    margin=<ltrb 0 0 0 0>
    outlineSettings {
     bUseBrushTransparency=false
     roundingType=FixedRadius
     width=1
     color {colorUseRule=UseColor_ColorTable, specifiedColor=<rgba 0 0 0 0>}
     cornerRadii=<3 3 3 3>
    }
    tintColor {colorUseRule=UseColor_Specified, specifiedColor=<rgba 0.0048 0.0048 0.0048 1>}
    uVRegion {bIsValid=false, max=<0 0>, min=<0 0>}
   }
   backgroundImage {
    bIsDynamicallyLoaded=false
    drawAs=RoundedBox
    imageType=NoImage
    mirroring=NoMirror
    resourceName=None
    resourceObject=null
    tiling=NoTile
    imageSize=<18 18>
    margin=<ltrb 0 0 0 0>
    outlineSettings {
     bUseBrushTransparency=false
     roundingType=FixedRadius
     width=1
     color {colorUseRule=UseColor_ColorTable, specifiedColor=<rgba 0 0 0 0>}
     cornerRadii=<3 3 3 3>
    }
    tintColor {colorUseRule=UseColor_Specified, specifiedColor=<rgba 0.0048 0.0048 0.0048 1>}
    uVRegion {bIsValid=false, max=<0 0>, min=<0 0>}
   }
   backgroundPressedImage {
    bIsDynamicallyLoaded=false
    drawAs=RoundedBox
    imageType=NoImage
    mirroring=NoMirror
    resourceName=None
    resourceObject=null
    tiling=NoTile
    imageSize=<18 18>
    margin=<ltrb 0 0 0 0>
    outlineSettings {
     bUseBrushTransparency=false
     roundingType=FixedRadius
     width=1
     color {colorUseRule=UseColor_ColorTable, specifiedColor=<rgba 0 0 0 0>}
     cornerRadii=<3 3 3 3>
    }
    tintColor {colorUseRule=UseColor_Specified, specifiedColor=<rgba 0.0048 0.0048 0.0048 1>}
    uVRegion {bIsValid=false, max=<0 0>, min=<0 0>}
   }
   borderBackgroundColor {colorUseRule=UseColor_Specified, specifiedColor=<rgba 1 1 1 1>}
   checkedForeground {colorUseRule=UseColor_Specified, specifiedColor=<rgba 1 1 1 1>}
   checkedHoveredForeground {colorUseRule=UseColor_Specified, specifiedColor=<rgba 1 1 1 1>}
   checkedHoveredImage {
    bIsDynamicallyLoaded=false
    drawAs=RoundedBox
    imageType=NoImage
    mirroring=NoMirror
    resourceName=None
    resourceObject=null
    tiling=NoTile
    imageSize=<0 0>
    margin=<ltrb 0 0 0 0>
    outlineSettings {
     bUseBrushTransparency=false
     roundingType=FixedRadius
     width=1
     color {colorUseRule=UseColor_Specified, specifiedColor=<rgba 0.006 0.006 0.006 1>}
     cornerRadii=<4 4 4 4>
    }
    tintColor {colorUseRule=UseColor_Specified, specifiedColor=<rgba 0.0044 0.2384 1 1>}
    uVRegion {bIsValid=false, max=<0 0>, min=<0 0>}
   }
   checkedImage {
    bIsDynamicallyLoaded=false
    drawAs=RoundedBox
    imageType=NoImage
    mirroring=NoMirror
    resourceName=None
    resourceObject=null
    tiling=NoTile
    imageSize=<0 0>
    margin=<ltrb 0 0 0 0>
    outlineSettings {
     bUseBrushTransparency=false
     roundingType=FixedRadius
     width=1
     color {colorUseRule=UseColor_Specified, specifiedColor=<rgba 0.006 0.006 0.006 1>}
     cornerRadii=<4 4 4 4>
    }
    tintColor {colorUseRule=UseColor_Specified, specifiedColor=<rgba 0 0.162 0.7454 1>}
    uVRegion {bIsValid=false, max=<0 0>, min=<0 0>}
   }
   checkedPressedForeground {colorUseRule=UseColor_Specified, specifiedColor=<rgba 1 1 1 1>}
   checkedPressedImage {
    bIsDynamicallyLoaded=false
    drawAs=RoundedBox
    imageType=NoImage
    mirroring=NoMirror
    resourceName=None
    resourceObject=null
    tiling=NoTile
    imageSize=<0 0>
    margin=<ltrb 0 0 0 0>
    outlineSettings {
     bUseBrushTransparency=false
     roundingType=FixedRadius
     width=1
     color {colorUseRule=UseColor_Specified, specifiedColor=<rgba 0.006 0.006 0.006 1>}
     cornerRadii=<4 4 4 4>
    }
    tintColor {colorUseRule=UseColor_Specified, specifiedColor=<rgba 0.0625 0.0625 0.0625 1>}
    uVRegion {bIsValid=false, max=<0 0>, min=<0 0>}
   }
   checkedSlateSound {resourceObject=null}
   foregroundColor {colorUseRule=UseColor_Specified, specifiedColor=<rgba 1 1 1 1>}
   hoveredForeground {colorUseRule=UseColor_Specified, specifiedColor=<rgba 1 1 1 1>}
   hoveredSlateSound {resourceObject=null}
   padding=<ltrb 8 5 8 5>
   pressedForeground {colorUseRule=UseColor_Specified, specifiedColor=<rgba 1 1 1 1>}
   uncheckedHoveredImage {
    bIsDynamicallyLoaded=false
    drawAs=RoundedBox
    imageType=NoImage
    mirroring=NoMirror
    resourceName=None
    resourceObject=null
    tiling=NoTile
    imageSize=<0 0>
    margin=<ltrb 0 0 0 0>
    outlineSettings {
     bUseBrushTransparency=false
     roundingType=FixedRadius
     width=1
     color {colorUseRule=UseColor_Specified, specifiedColor=<rgba 0.006 0.006 0.006 1>}
     cornerRadii=<4 4 4 4>
    }
    tintColor {colorUseRule=UseColor_Specified, specifiedColor=<rgba 0.0953 0.0953 0.0953 1>}
    uVRegion {bIsValid=false, max=<0 0>, min=<0 0>}
   }
   uncheckedImage {
    bIsDynamicallyLoaded=false
    drawAs=RoundedBox
    imageType=NoImage
    mirroring=NoMirror
    resourceName=None
    resourceObject=null
    tiling=NoTile
    imageSize=<0 0>
    margin=<ltrb 0 0 0 0>
    outlineSettings {
     bUseBrushTransparency=false
     roundingType=FixedRadius
     width=1
     color {colorUseRule=UseColor_Specified, specifiedColor=<rgba 0.006 0.006 0.006 1>}
     cornerRadii=<4 4 4 4>
    }
    tintColor {colorUseRule=UseColor_Specified, specifiedColor=<rgba 0.0284 0.0284 0.0284 1>}
    uVRegion {bIsValid=false, max=<0 0>, min=<0 0>}
   }
   uncheckedPressedImage {
    bIsDynamicallyLoaded=false
    drawAs=RoundedBox
    imageType=NoImage
    mirroring=NoMirror
    resourceName=None
    resourceObject=null
    tiling=NoTile
    imageSize=<0 0>
    margin=<ltrb 0 0 0 0>
    outlineSettings {
     bUseBrushTransparency=false
     roundingType=FixedRadius
     width=1
     color {colorUseRule=UseColor_Specified, specifiedColor=<rgba 0.006 0.006 0.006 1>}
     cornerRadii=<4 4 4 4>
    }
    tintColor {colorUseRule=UseColor_Specified, specifiedColor=<rgba 0.0395 0.0395 0.0395 1>}
    uVRegion {bIsValid=false, max=<0 0>, min=<0 0>}
   }
   uncheckedSlateSound {resourceObject=null}
   undeterminedForeground {colorUseRule=UseColor_Specified, specifiedColor=<rgba 1 1 1 1>}
   undeterminedHoveredImage {
    bIsDynamicallyLoaded=false
    drawAs=Image
    imageType=Vector
    mirroring=NoMirror
    resourceName=../../../Engine/Content/Slate/Starship/CoreWidgets/CheckBox/indeterminate.svg
    resourceObject=null
    tiling=NoTile
    imageSize=<16 16>
    margin=<ltrb 0 0 0 0>
    outlineSettings {
     bUseBrushTransparency=false
     roundingType=HalfHeightRadius
     width=0
     color {colorUseRule=UseColor_Specified, specifiedColor=<rgba 0 0 0 0>}
     cornerRadii=<0 0 0 0>
    }
    tintColor {colorUseRule=UseColor_Specified, specifiedColor=<rgba 0.0044 0.2384 1 1>}
    uVRegion {bIsValid=false, max=<0 0>, min=<0 0>}
   }
   undeterminedImage {
    bIsDynamicallyLoaded=false
    drawAs=Image
    imageType=Vector
    mirroring=NoMirror
    resourceName=../../../Engine/Content/Slate/Starship/CoreWidgets/CheckBox/indeterminate.svg
    resourceObject=null
    tiling=NoTile
    imageSize=<16 16>
    margin=<ltrb 0 0 0 0>
    outlineSettings {
     bUseBrushTransparency=false
     roundingType=HalfHeightRadius
     width=0
     color {colorUseRule=UseColor_Specified, specifiedColor=<rgba 0 0 0 0>}
     cornerRadii=<0 0 0 0>
    }
    tintColor {colorUseRule=UseColor_Specified, specifiedColor=<rgba 0 0.162 0.7454 1>}
    uVRegion {bIsValid=false, max=<0 0>, min=<0 0>}
   }
   undeterminedPressedImage {
    bIsDynamicallyLoaded=false
    drawAs=Image
    imageType=Vector
    mirroring=NoMirror
    resourceName=../../../Engine/Content/Slate/Starship/CoreWidgets/CheckBox/indeterminate.svg
    resourceObject=null
    tiling=NoTile
    imageSize=<16 16>
    margin=<ltrb 0 0 0 0>
    outlineSettings {
     bUseBrushTransparency=false
     roundingType=HalfHeightRadius
     width=0
     color {colorUseRule=UseColor_Specified, specifiedColor=<rgba 0 0 0 0>}
     cornerRadii=<0 0 0 0>
    }
    tintColor {colorUseRule=UseColor_Specified, specifiedColor=<rgba 0 0.0802 0.3515 1>}
    uVRegion {bIsValid=false, max=<0 0>, min=<0 0>}
   }
  }
 }
 Checked State {type=ECheckBoxState, default=Unchecked}
 Checked Tool Tip {type=string, default=""}
 Default Text Color {type=SlateColor, default {colorUseRule=UseColor_Specified, specifiedColor=<rgba 0.5271 0.5271 0.5271 1>}}
 Highlighted Text Color {type=SlateColor, default {colorUseRule=UseColor_Specified, specifiedColor=<rgba 1 1 1 1>}}
 Text {type=string, default=Text}
 Unchecked Tool Tip {type=string, default=""}
]
widgets {
 info {
  namedSlotCount=1
  rootWidgetClass=&/EUW_ToggleButton_ContentSlot.EUW_ToggleButton_ContentSlot_C
  widgetCount=3
 }
 widgets [
  ToggleButton_Content {
   widgetClassPath=&/EUW_ToggleButton_ContentSlot
   bIsVariable=true
   widget=ToggleButton_Content
   properties {check Box Style {
     checkBoxType=ToggleButton
     backgroundHoveredImage {
      bIsDynamicallyLoaded=false
      drawAs=RoundedBox
      imageType=NoImage
      mirroring=NoMirror
      resourceName=None
      resourceObject=null
      tiling=NoTile
      imageSize=<18 18>
      margin=<ltrb 0 0 0 0>
      outlineSettings {
       bUseBrushTransparency=false
       roundingType=FixedRadius
       width=1
       color {colorUseRule=UseColor_ColorTable, specifiedColor=<rgba 0 0 0 0>}
       cornerRadii=<3 3 3 3>
      }
      tintColor {colorUseRule=UseColor_Specified, specifiedColor=<rgba 0.0048 0.0048 0.0048 1>}
      uVRegion {bIsValid=false, max=<0 0>, min=<0 0>}
     }
     backgroundImage {
      bIsDynamicallyLoaded=false
      drawAs=RoundedBox
      imageType=NoImage
      mirroring=NoMirror
      resourceName=None
      resourceObject=null
      tiling=NoTile
      imageSize=<18 18>
      margin=<ltrb 0 0 0 0>
      outlineSettings {
       bUseBrushTransparency=false
       roundingType=FixedRadius
       width=1
       color {colorUseRule=UseColor_ColorTable, specifiedColor=<rgba 0 0 0 0>}
       cornerRadii=<3 3 3 3>
      }
      tintColor {colorUseRule=UseColor_Specified, specifiedColor=<rgba 0.0048 0.0048 0.0048 1>}
      uVRegion {bIsValid=false, max=<0 0>, min=<0 0>}
     }
     backgroundPressedImage {
      bIsDynamicallyLoaded=false
      drawAs=RoundedBox
      imageType=NoImage
      mirroring=NoMirror
      resourceName=None
      resourceObject=null
      tiling=NoTile
      imageSize=<18 18>
      margin=<ltrb 0 0 0 0>
      outlineSettings {
       bUseBrushTransparency=false
       roundingType=FixedRadius
       width=1
       color {colorUseRule=UseColor_ColorTable, specifiedColor=<rgba 0 0 0 0>}
       cornerRadii=<3 3 3 3>
      }
      tintColor {colorUseRule=UseColor_Specified, specifiedColor=<rgba 0.0048 0.0048 0.0048 1>}
      uVRegion {bIsValid=false, max=<0 0>, min=<0 0>}
     }
     borderBackgroundColor {colorUseRule=UseColor_Specified, specifiedColor=<rgba 1 1 1 1>}
     checkedForeground {colorUseRule=UseColor_Specified, specifiedColor=<rgba 1 1 1 1>}
     checkedHoveredForeground {colorUseRule=UseColor_Specified, specifiedColor=<rgba 1 1 1 1>}
     checkedHoveredImage {
      bIsDynamicallyLoaded=false
      drawAs=RoundedBox
      imageType=NoImage
      mirroring=NoMirror
      resourceName=None
      resourceObject=null
      tiling=NoTile
      imageSize=<0 0>
      margin=<ltrb 0 0 0 0>
      outlineSettings {
       bUseBrushTransparency=false
       roundingType=FixedRadius
       width=1
       color {colorUseRule=UseColor_Specified, specifiedColor=<rgba 0.006 0.006 0.006 1>}
       cornerRadii=<4 4 4 4>
      }
      tintColor {colorUseRule=UseColor_Specified, specifiedColor=<rgba 0.0044 0.2384 1 1>}
      uVRegion {bIsValid=false, max=<0 0>, min=<0 0>}
     }
     checkedImage {
      bIsDynamicallyLoaded=false
      drawAs=RoundedBox
      imageType=NoImage
      mirroring=NoMirror
      resourceName=None
      resourceObject=null
      tiling=NoTile
      imageSize=<0 0>
      margin=<ltrb 0 0 0 0>
      outlineSettings {
       bUseBrushTransparency=false
       roundingType=FixedRadius
       width=1
       color {colorUseRule=UseColor_Specified, specifiedColor=<rgba 0.006 0.006 0.006 1>}
       cornerRadii=<4 4 4 4>
      }
      tintColor {colorUseRule=UseColor_Specified, specifiedColor=<rgba 0 0.162 0.7454 1>}
      uVRegion {bIsValid=false, max=<0 0>, min=<0 0>}
     }
     checkedPressedForeground {colorUseRule=UseColor_Specified, specifiedColor=<rgba 1 1 1 1>}
     checkedPressedImage {
      bIsDynamicallyLoaded=false
      drawAs=RoundedBox
      imageType=NoImage
      mirroring=NoMirror
      resourceName=None
      resourceObject=null
      tiling=NoTile
      imageSize=<0 0>
      margin=<ltrb 0 0 0 0>
      outlineSettings {
       bUseBrushTransparency=false
       roundingType=FixedRadius
       width=1
       color {colorUseRule=UseColor_Specified, specifiedColor=<rgba 0.006 0.006 0.006 1>}
       cornerRadii=<4 4 4 4>
      }
      tintColor {colorUseRule=UseColor_Specified, specifiedColor=<rgba 0.0625 0.0625 0.0625 1>}
      uVRegion {bIsValid=false, max=<0 0>, min=<0 0>}
     }
     checkedSlateSound {resourceObject=null}
     foregroundColor {colorUseRule=UseColor_Specified, specifiedColor=<rgba 1 1 1 1>}
     hoveredForeground {colorUseRule=UseColor_Specified, specifiedColor=<rgba 1 1 1 1>}
     hoveredSlateSound {resourceObject=null}
     padding=<ltrb 8 5 8 5>
     pressedForeground {colorUseRule=UseColor_Specified, specifiedColor=<rgba 1 1 1 1>}
     uncheckedHoveredImage {
      bIsDynamicallyLoaded=false
      drawAs=RoundedBox
      imageType=NoImage
      mirroring=NoMirror
      resourceName=None
      resourceObject=null
      tiling=NoTile
      imageSize=<0 0>
      margin=<ltrb 0 0 0 0>
      outlineSettings {
       bUseBrushTransparency=false
       roundingType=FixedRadius
       width=1
       color {colorUseRule=UseColor_Specified, specifiedColor=<rgba 0.006 0.006 0.006 1>}
       cornerRadii=<4 4 4 4>
      }
      tintColor {colorUseRule=UseColor_Specified, specifiedColor=<rgba 0.0953 0.0953 0.0953 1>}
      uVRegion {bIsValid=false, max=<0 0>, min=<0 0>}
     }
     uncheckedImage {
      bIsDynamicallyLoaded=false
      drawAs=RoundedBox
      imageType=NoImage
      mirroring=NoMirror
      resourceName=None
      resourceObject=null
      tiling=NoTile
      imageSize=<0 0>
      margin=<ltrb 0 0 0 0>
      outlineSettings {
       bUseBrushTransparency=false
       roundingType=FixedRadius
       width=1
       color {colorUseRule=UseColor_Specified, specifiedColor=<rgba 0.006 0.006 0.006 1>}
       cornerRadii=<4 4 4 4>
      }
      tintColor {colorUseRule=UseColor_Specified, specifiedColor=<rgba 0.0284 0.0284 0.0284 1>}
      uVRegion {bIsValid=false, max=<0 0>, min=<0 0>}
     }
     uncheckedPressedImage {
      bIsDynamicallyLoaded=false
      drawAs=RoundedBox
      imageType=NoImage
      mirroring=NoMirror
      resourceName=None
      resourceObject=null
      tiling=NoTile
      imageSize=<0 0>
      margin=<ltrb 0 0 0 0>
      outlineSettings {
       bUseBrushTransparency=false
       roundingType=FixedRadius
       width=1
       color {colorUseRule=UseColor_Specified, specifiedColor=<rgba 0.006 0.006 0.006 1>}
       cornerRadii=<4 4 4 4>
      }
      tintColor {colorUseRule=UseColor_Specified, specifiedColor=<rgba 0.0395 0.0395 0.0395 1>}
      uVRegion {bIsValid=false, max=<0 0>, min=<0 0>}
     }
     uncheckedSlateSound {resourceObject=null}
     undeterminedForeground {colorUseRule=UseColor_Specified, specifiedColor=<rgba 1 1 1 1>}
     undeterminedHoveredImage {
      bIsDynamicallyLoaded=false
      drawAs=Image
      imageType=Vector
      mirroring=NoMirror
      resourceName=../../../Engine/Content/Slate/Starship/CoreWidgets/CheckBox/indeterminate.svg
      resourceObject=null
      tiling=NoTile
      imageSize=<16 16>
      margin=<ltrb 0 0 0 0>
      outlineSettings {
       bUseBrushTransparency=false
       roundingType=HalfHeightRadius
       width=0
       color {colorUseRule=UseColor_Specified, specifiedColor=<rgba 0 0 0 0>}
       cornerRadii=<0 0 0 0>
      }
      tintColor {colorUseRule=UseColor_Specified, specifiedColor=<rgba 0.0044 0.2384 1 1>}
      uVRegion {bIsValid=false, max=<0 0>, min=<0 0>}
     }
     undeterminedImage {
      bIsDynamicallyLoaded=false
      drawAs=Image
      imageType=Vector
      mirroring=NoMirror
      resourceName=../../../Engine/Content/Slate/Starship/CoreWidgets/CheckBox/indeterminate.svg
      resourceObject=null
      tiling=NoTile
      imageSize=<16 16>
      margin=<ltrb 0 0 0 0>
      outlineSettings {
       bUseBrushTransparency=false
       roundingType=HalfHeightRadius
       width=0
       color {colorUseRule=UseColor_Specified, specifiedColor=<rgba 0 0 0 0>}
       cornerRadii=<0 0 0 0>
      }
      tintColor {colorUseRule=UseColor_Specified, specifiedColor=<rgba 0 0.162 0.7454 1>}
      uVRegion {bIsValid=false, max=<0 0>, min=<0 0>}
     }
     undeterminedPressedImage {
      bIsDynamicallyLoaded=false
      drawAs=Image
      imageType=Vector
      mirroring=NoMirror
      resourceName=../../../Engine/Content/Slate/Starship/CoreWidgets/CheckBox/indeterminate.svg
      resourceObject=null
      tiling=NoTile
      imageSize=<16 16>
      margin=<ltrb 0 0 0 0>
      outlineSettings {
       bUseBrushTransparency=false
       roundingType=HalfHeightRadius
       width=0
       color {colorUseRule=UseColor_Specified, specifiedColor=<rgba 0 0 0 0>}
       cornerRadii=<0 0 0 0>
      }
      tintColor {colorUseRule=UseColor_Specified, specifiedColor=<rgba 0 0.0802 0.3515 1>}
      uVRegion {bIsValid=false, max=<0 0>, min=<0 0>}
     }
    }}
  }
  VerticalBox_0 {widgetClassPath=&VerticalBox, namedSlotHost=ToggleButton_Content, widget=VerticalBox_0}
  TextBlock {
   widgetClassPath=&TextBlock
   bIsVariable=true
   parent=VerticalBox_0
   widget=TextBlock
   properties {
    justification=Center
    text=Text
    font {
     bForceMonospaced=false
     bMaterialIsStencil=false
     fontMaterial=null
     fontObject=&Roboto
     letterSpacing=0
     monospacedWidth=1
     size=8.25
     skewAmount=0
     typefaceFontName=Regular
     outlineSettings {
      bApplyOutlineToDropShadows=false
      bMiteredCorners=false
      bSeparateFillAlpha=false
      outlineMaterial=null
      outlineSize=0
      outlineColor=<rgba 0 0 0 1>
     }
    }
   }
   slot {class=VerticalBoxSlot, properties {verticalAlignment=VAlign_Center, size {sizeRule=Fill, value=1}}}
  }
 ]
}
