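# Editor utility widget for the ElevenLabs VO Generator, an editor tool that generates voice-over wave files and optional transcriptions and ratings from CSV datasets and voice configurations.
# A preset data asset fills the sample rate, bit depth, configurations and paths.
# RefreshOutputPanel validates settings and shows clearable status labels.
# Generate runs embedded Python that checks the API key and inputs, then passes a base64 JSON payload to eltests.py via uv in a new console.
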EditorUtilityWidgetBlueprint /FortTechAudioTools/Widgets/AIVoiceGenerator/EUW_AIVoiceGenerator
graphs [
 AddErrorMessage {nodes [
   n1 = AddErrorMessage!() -> (then, Message:Text) {is=title K2Node_FunctionEntry}
   > n2 = Create Widget!(
    execute:Exec
    WorldContextObject:Object Reference
    Class:User Widget Class Reference=&/EUW_ClearableInfoLabel.EUW_ClearableInfoLabel_C
    OwningPlayer:Player Controller Object Reference
    FullIconPath:Name=CoreStyle.Icons.Error.Large
    IconSize:Vector 2D Structure=Get OutputIconSize() -> (OutputIconSize:Vector 2D Structure)
    IconColor:Linear Color Structure=(R=1,G=0.5365,B=0.5527,A=1)
    Text:Text=@n1
    FontSize:double=Get FontSize() -> (FontSize:double)
    TextWidth:float=Get TextWidth() -> (TextWidth:float)
    TextColor:Linear Color Structure=(R=1,G=1,B=1,A=1)
    BackgroundColor:Linear Color Structure=(R=0.3906,G=0,B=0.0062,A=1)
   ) -> (then:Exec, ReturnValue:EUW Clearable Info Label Object Reference)
   Bind Event to On Clear Clicked(self:EUW Clearable Info Label Object Reference=@n2) {is=title K2Node_AddDelegate}
   AddChild(Content=@n2)
   Create Event()
  ]}
 AddLoadingMessage {nodes [
   n1 = AddLoadingMessage!() -> (then, Message:Text) {is=title K2Node_FunctionEntry}
   > n2 = Create Widget(
    Class=&EUW_ClearableInfoLabel_C
    FullIconPath=CoreStyle.NotificationList.Throbber
    IconSize=Get OutputIconSize()
    IconColor=(R=0.9446,G=0.7606,B=1,A=1)
    Text=@n1
    FontSize=Get FontSize()
    TextWidth=Get TextWidth()
    TextColor=(R=1,G=1,B=1,A=1)
    BackgroundColor=(R=0.0618,G=0.0307,B=0.151,A=1)
   )
   > Add(
    TargetArray:Array of EUW Clearable Info Label Object References=Get LoadingInfoLabels() -> (LoadingInfoLabels:Array of EUW Clearable Info Label Object References)
    NewItem:"EUW Clearable Info Label Object Reference (by ref)"=@n2
   ) {is=title K2Node_CallArrayFunction}
   Bind Event to On Clear Clicked(self=@n2)
   AddChild(Content=@n2)
  ]}
 AddSuccessMessage {nodes [
   n1 = AddSuccessMessage!() -> (then, Message:Text) {is=title K2Node_FunctionEntry}
   > n2 = Create Widget(
    Class=&EUW_ClearableInfoLabel_C
    FullIconPath=CoreStyle.Icons.Success.Large
    IconSize=Get OutputIconSize()
    IconColor=(R=0.1983,G=0.8073,B=0.1976,A=1)
    Text=@n1
    FontSize=Get FontSize()
    TextWidth=Get TextWidth()
    TextColor=(R=1,G=1,B=1,A=1)
    BackgroundColor=(R=0.0099,G=0.1042,B=0.0162,A=1)
   )
   Bind Event to On Clear Clicked(self=@n2)
   AddChild(Content=@n2)
  ]}
 AddWarningMessage {nodes [
   n1 = AddWarningMessage!() -> (then, Message:Text) {is=title K2Node_FunctionEntry}
   > n2 = Create Widget(
    Class=&EUW_ClearableInfoLabel_C
    FullIconPath=CoreStyle.Icons.Error.Large
    IconSize=Get OutputIconSize()
    IconColor=(R=1,G=0.2223,B=0,A=1)
    Text=@n1
    FontSize=Get FontSize()
    TextWidth=Get TextWidth()
    TextColor=(R=0,G=0,B=0,A=1)
    BackgroundColor=(R=1,G=0.7126,B=0.2678,A=1)
   )
   Bind Event to On Clear Clicked(self=@n2)
   AddChild(Content=@n2)
   Create Event()
  ]}
 ApplySettingsFromPreset {nodes [
   ApplySettingsFromPreset() {is=title K2Node_FunctionEntry}
   > Get!(execute:Exec) -> (then:Exec, else:Exec, GenerationSettingsPreset:BP AIVoice Generator Preset Data Asset Object Reference)
   n1 = Get GenerationSettingsPreset() -> (GenerationSettingsPreset:BP AIVoice Generator Preset Data Asset Object Reference)
   > Set with Broadcast SampleRate!(
    execute
    SampleRate:EAIVoiceGeneratorSampleRate Enum=Get SampleRate!(self:BP AIVoice Generator Preset Data Asset Object Reference=@n1) -> (SampleRate:EAIVoiceGeneratorSampleRate Enum=NewEnumerator0)
   ) -> (then, Output_Get:EAIVoiceGeneratorSampleRate Enum=NewEnumerator0)
   > Set with Broadcast BitDepth!(
    execute
    BitDepth:EAIVoiceGeneratorBitDepth Enum=Get BitDepth!(self:BP AIVoice Generator Preset Data Asset Object Reference=@n1) -> (BitDepth:EAIVoiceGeneratorBitDepth Enum=NewEnumerator0)
   ) -> (then, Output_Get:EAIVoiceGeneratorBitDepth Enum=NewEnumerator0)
   > Set with Broadcast SelectedConfigurations!(
    execute
    SelectedConfigurations:Array of Strings=Get FilteredConfigurations!(self:BP AIVoice Generator Preset Data Asset Object Reference=@n1) -> (FilteredConfigurations:Array of Strings)
   ) -> (then, Output_Get:Array of Strings)
   > SetSelectedFilePath!(
    execute
    self:Fort Tech Audio Tools Path Selector Object Reference=Get EUW_PathSelector_Dataset() -> (EUW_PathSelector_Dataset:Fort Tech Audio Tools Path Selector Object Reference)
    SelectedFilePath:String=Get DatasetPath!(self:BP AIVoice Generator Preset Data Asset Object Reference=@n1) -> (DatasetPath:String)
   ) -> (then)
   > SetSelectedFilePath(
    self=Get EUW_PathSelector_Configuration() -> (EUW_PathSelector_Configuration:Fort Tech Audio Tools Path Selector Object Reference)
    SelectedFilePath=Get ConfigurationsPath!(self:BP AIVoice Generator Preset Data Asset Object Reference=@n1) -> (ConfigurationsPath:String)
   )
   > SetSelectedFilePath(
    self=Get EUW_PathSelector_GeneratedWaveFileDirectory() -> (EUW_PathSelector_GeneratedWaveFileDirectory:Fort Tech Audio Tools Path Selector Object Reference)
    SelectedFilePath=Get OutputDirectory!(self:BP AIVoice Generator Preset Data Asset Object Reference=@n1) -> (OutputDirectory:String)
   )
   Is Toggle Active!(Target:Tech Audio Tools Toggle Button Text Object Reference) -> (ReturnValue:bool=~) {is=title K2Node_MacroInstance}
   Get GenerateSpeechDataset!(self:BP AIVoice Generator Preset Data Asset Object Reference=@n1) -> (GenerateSpeechDataset:bool)
   Get GenerateTranscriptions!(self:BP AIVoice Generator Preset Data Asset Object Reference=@n1) -> (GenerateTranscriptions:bool)
   Get GenerateRatings!(self:BP AIVoice Generator Preset Data Asset Object Reference=@n1) -> (GenerateRatings:bool)
   Get GatherTranscriptions!(self:BP AIVoice Generator Preset Data Asset Object Reference=@n1) -> (GatherTranscriptions:bool)
  ]}
 CheckSelectedConfigurations {nodes [
   CheckSelectedConfigurations() {is=title K2Node_FunctionEntry}
   > Clear(TargetArray=Get ValidConfigurationNames() -> (ValidConfigurationNames:Array of Strings))
   > n1 = For Each Loop(Array=Get SelectedConfigurations() -> (SelectedConfigurations:Array of Strings))
   Return Node!(
    execute=@n1.Completed
    ReturnValue:bool=Is Not Empty!(TargetArray=Get ValidConfigurations() -> (ValidConfigurations:Array of Strings)) -> (ReturnValue:bool)
   )
   > n2 = Branch(Condition=IsEmpty(InString=@n1))
   n3 = FindHeaderInCSVFile!(
    execute=@n2.else
    CsvFilePath:String=GetSelectedFilePath!(self:Fort Tech Audio Tools Path Selector Object Reference=Get EUW_PathSelector_Configuration()) -> (SelectedFilePath:String)
    HeaderName:String=@n1
    __WorldContext:Object Reference
   ) -> (then, Found:bool, OutputMsg:String)
   > Branch(Condition=@n3)
   > Add(TargetArray:Array of Strings=Get ValidConfigurations(), NewItem:"String (by ref)"=@n1)
   AddWarningMessage!(execute, self, Message:Text=Format Text!(Format="LOCTEXT(\"Configuration \\'{i}\\' is empty.\")", i:int=@n1.Array Index) -> (Result)) -> (then) {is=title K2Node_CallFunction}
   AddWarningMessage(Message="To Text (String)"(InString=@n3.OutputMsg))
   Get ValidConfigurationNames()
  ]}
 EventGraph {nodes [
   "On Clicked (EUW_Button_Primary_GenerateSourceAssets)"() {is=title K2Node_ComponentBoundEvent}
   > GenerateSourceAssets()
   "On Property Changed (DetailsView_Preset)"!() -> (OutputDelegate, then, PropertyName:Name=None) {is=title K2Node_ComponentBoundEvent}
   > ApplySettingsFromPreset() {is=title K2Node_CallFunction}
   n1 = OnSelectedFilePathChanged_CSV!() -> (OutputDelegate, then, SelectedFilePath:String) {is=title K2Node_CustomEvent}
   > SetInitialSearchPath!(
    execute
    self:Fort Tech Audio Tools Path Selector Object Reference=[Get EUW_PathSelector_Dataset(), Get EUW_PathSelector_Configuration()]
    InitialSearchPath:String=@n1.SelectedFilePath
   ) -> (then)
   n2 = OnGeneratedAssetDirectoryChanged!() -> (OutputDelegate, then, SelectedFilePath:String) {is=title K2Node_CustomEvent}
   > SetInitialSearchPath(self=Get EUW_PathSelector_GeneratedWaveFileDirectory(), InitialSearchPath=@n2.SelectedFilePath)
   OnTimerUpdate() {is=title K2Node_CustomEvent}
   > Branch(Condition=Is Not Empty(TargetArray:Array of EUW Clearable Info Label Object References=Get LoadingInfoLabels()))
   n3 = FInterpTo(Current=Get CurrentAngle() -> (CurrentAngle:double), Target=float + float(A=Get CurrentAngle(), B=10), DeltaTime=0.01, InterpSpeed=100)
   > Set CurrentAngle(CurrentAngle:double=@n3) -> (Output_Get:double)
   > n4 = For Each Loop(Array:Array of EUW Clearable Info Label Object References=Get LoadingInfoLabels()) -> (Array Element:EUW Clearable Info Label Object Reference)
   > SetRenderTransformAngle!(
    execute
    self:Widget Object Reference=Get TechAudioToolsEditorSlateIcon!(self:EUW Clearable Info Label Object Reference=@n4) -> (TechAudioToolsEditorSlateIcon:Tech Audio Tools Editor Slate Icon Object Reference)
    Angle:float=@n3
   ) -> (then)
   Event Construct()
   > Bind Event to On Selected File Path Changed(
    self:Fort Tech Audio Tools Path Selector Object Reference=[Get EUW_PathSelector_Dataset(), Get EUW_PathSelector_Configuration()]
    Delegate=Create Event()
   ) {is=title K2Node_AddDelegate}
   > Bind Event to On Selected File Path Changed(self=Get EUW_PathSelector_GeneratedWaveFileDirectory(), Delegate=Create Event())
   Event PreConstruct()
   > SetObject(
    self=[
     Get DetailsView_SampleRate() -> (DetailsView_SampleRate:Details View Object Reference)
     Get DetailsView_Preset() -> (DetailsView_Preset:Details View Object Reference)
     Get DetailsView_Configurations() -> (DetailsView_Configurations:Details View Object Reference)
    ]
    NewObject=Self-Reference()
   )
   RefreshOutputPanel()
   RefreshOutputPanel()
   RefreshOutputPanel()
   RefreshOutputPanel()
   Set Timer by Event(Delegate=Create Event(), Time=Get UpdateInterval() -> (UpdateInterval:float), bLooping=true, bMaxOncePerFrame=true)
   "On Property Changed (DetailsView_Configurations)"!() -> (OutputDelegate, then, PropertyName:Name=None) {is=title K2Node_ComponentBoundEvent}
  ]}
 GenerateSourceAssets {nodes [
   n1 = CheckSelectedConfigurations!(execute, self) -> (then, ReturnValue:bool) {is=title K2Node_CallFunction}
   > n2 = Branch(Condition=@n1)
   n3 = GetSelectedFilePath(self=Get EUW_PathSelector_Configuration())
   n4 = Execute Python Script!(
    execute=[@n6.then, @n2.then]
    PythonScript="## Copyright Epic Games, Inc. All Rights Reserved.\r\n\r\nimport os\r\nimport sys\r\nimport csv\r\nimport unreal\r\nimport json\r\nimport base64\r\nimport asyncio\r\nimport subprocess\r\nimport shutil\r\nfrom pathlib import Path\r\n\r\n## Type-hint values passed in from Blueprints.\r\ngenerate_speech_dataset: bool\r\ngenerate_transcriptions: bool\r\ngenerate_ratings: bool\r\ngather_transcriptions: bool\r\nfiltered_configurations: list[str]\r\ncsv_file_path_dataset: str\r\ncsv_file_path_configurations: str\r\ncsv_file_path_transcriptions_for_rating: str\r\noutput_directory: str\r\nsamplerate: int\r\nbitdepthindex:int\r\n\r\nSUPPORTED_SAMPLERATES: list[int] = [\r\n    8000,\r\n    11025,\r\n    16000,\r\n    22050,\r\n    24000,\r\n    44100,\r\n    48000\r\n]\r\n\r\n## TODO - This seems wrong. Use pathlib.Path and investigate further. Prefer no hard-coded FortTechAudioTools path.\r\n## Cannot use __file__ value to get current path when running python through Unreal's 'Execute Python Script' node.\r\n## Hard-code FortTechAudioTools' python directory.\r\nscript_path:str = unreal.Paths.convert_relative_path_to_full(unreal.Paths.project_plugins_dir() + \"FortTechAudioTools/Content/Python/ElevenLabs/\")\r\nsys.path.append(script_path)\r\n\r\n\r\nasync def run_uv_subprocess(json_arguments_b64:str) -> int | None:\r\n    \r\n    uv_subprocess = await asyncio.create_subprocess_exec(\r\n        \"cmd.exe\", f\"/c uv run eltests.py --payload {json_arguments_b64}\",\r\n        cwd=script_path,\r\n        creationflags=subprocess.CREATE_NEW_CONSOLE,\r\n    )\r\n    \r\n    await uv_subprocess.wait()\r\n    \r\n\r\ndef main():\r\n    \r\n    if not os.getenv(\"ELEVENLABS_API_KEY\"):\r\n        unreal.EditorDialog.show_message(unreal.Text(\"Error\"), unreal.Text(\"This script requires an ElevenLabs API key. Please add an 'ELEVENLABS_API_KEY' environment variable to your system.\"), unreal.AppMsgType.OK, unreal.AppReturnType.NO, unreal.AppMsgCategory.ERROR)\r\n        raise RuntimeError(f\"This script requires an ElevenLabs API key. Please add an 'ELEVENLABS_API_KEY' environment variable to your system.\")\r\n    \r\n    if shutil.which(\"uv\") is None:\r\n        unreal.EditorDialog.show_message(unreal.Text(\"Error\"), unreal.Text(\"UV executable not found. Please install UV to execute this script: (https://docs.astral.sh/uv/).\"), unreal.AppMsgType.OK, unreal.AppReturnType.NO, unreal.AppMsgCategory.ERROR)\r\n        raise RuntimeError(f\"UV executable not found. Please install UV to execute this script: https://docs.astral.sh/uv/\")\r\n    \r\n    if generate_speech_dataset and not csv_file_path_dataset:\r\n        unreal.EditorDialog.show_message(unreal.Text(\"Error\"), unreal.Text(\"Dataset file path is required.\"), unreal.AppMsgType.OK, unreal.AppReturnType.NO, unreal.AppMsgCategory.ERROR)\r\n        raise ValueError(f\"csv_file_path_dataset is required.\")\r\n    \r\n    if generate_speech_dataset and not csv_file_path_configurations:\r\n        unreal.EditorDialog.show_message(unreal.Text(\"Error\"), unreal.Text(\"Configuration file path is required.\"), unreal.AppMsgType.OK, unreal.AppReturnType.NO, unreal.AppMsgCategory.ERROR)\r\n        raise ValueError(f\"csv_file_path_configurations is required.\")\r\n    \r\n    if generate_speech_dataset and len(filtered_configurations) <= 0:\r\n        unreal.EditorDialog.show_message(unreal.Text(\"Error\"), unreal.Text(\"At least one configuration must be selected.\"), unreal.AppMsgType.OK, unreal.AppReturnType.NO, unreal.AppMsgCategory.ERROR)\r\n        raise ValueError(f\"No configurations selected.\")\r\n    \r\n    ## Check provided configuration names against headers of .csv file.\r\n    ## Exit if even one configuration name cannot be found.\r\n    with Path(csv_file_path_configurations).open() as csv_file:\r\n        csv_reader = csv.DictReader(csv_file, delimiter=',')\r\n        for configuration in filtered_configurations:\r\n            if configuration not in csv_reader.fieldnames:\r\n                unreal.EditorDialog.show_message(unreal.Text(\"Error\"), unreal.Text(f\"Configuration not found in dataset: '{configuration}'.\"), unreal.AppMsgType.OK, unreal.AppReturnType.NO, unreal.AppMsgCategory.ERROR)\r\n                raise ValueError(f\"Configuration name not detected in .csv file: '{configuration}'\")\r\n    \r\n    if generate_speech_dataset and not filtered_configurations:\r\n        unreal.EditorDialog.show_message(unreal.Text(\"Error\"), unreal.Text(\"At least one configuration must be selected.\"), unreal.AppMsgType.OK, unreal.AppReturnType.NO, unreal.AppMsgCategory.ERROR)\r\n        raise ValueError(f\"No configurations selected.\")\r\n    \r\n    if generate_transcriptions and not csv_file_path_transcriptions_for_rating:\r\n        unreal.EditorDialog.show_message(unreal.Text(\"Error\"), unreal.Text(\"csv_file_path_transcriptions_for_rating is required.\"), unreal.AppMsgType.OK, unreal.AppReturnType.NO, unreal.AppMsgCategory.ERROR)\r\n        raise ValueError(f\"csv_file_path_transcriptions_for_rating is required.\")\r\n    \r\n    if not output_directory:\r\n        unreal.EditorDialog.show_message(unreal.Text(\"Error\"), unreal.Text(\"Output directory is required.\"), unreal.AppMsgType.OK, unreal.AppReturnType.NO, unreal.AppMsgCategory.ERROR)\r\n        raise ValueError(f\"Output directory is required.\")\r\n    \r\n    if generate_speech_dataset and samplerate not in SUPPORTED_SAMPLERATES:\r\n        unreal.EditorDialog.show_message(unreal.Text(\"Error\"), unreal.Text(\"Sample Rate not supported.\"), unreal.AppMsgType.OK, unreal.AppReturnType.NO, unreal.AppMsgCategory.ERROR)\r\n        raise ValueError(f\"Sample Rate not supported.\")\r\n    \r\n    ## Store Blueprint variables in json structure as a lightweight way to redirect values to eltests.py\r\n    arguments = {\r\n        \"generate_speech_dataset\": generate_speech_dataset,\r\n        \"generate_transcriptions\": generate_transcriptions,\r\n        \"generate_ratings\": generate_ratings,\r\n        \"gather_transcriptions\": gather_transcriptions,\r\n        \"filtered_configurations\": list(filtered_configurations),\r\n        \"csv_file_path_dataset\": csv_file_path_dataset,\r\n        \"csv_file_path_configurations\": csv_file_path_configurations,\r\n        \"csv_file_path_transcriptions_for_rating\": csv_file_path_transcriptions_for_rating,\r\n        \"output_directory\": output_directory,\r\n        \"samplerate\": samplerate,\r\n        \"samplewidth\": bitdepthindex\r\n    }\r\n    json_arguments_bytes = json.dumps(arguments).encode(\"utf-8\")\r\n    json_arguments_b64 = base64.b64encode(json_arguments_bytes).decode(\"ascii\")\r\n    \r\n    asyncio.run(run_uv_subprocess(json_arguments_b64))\r\n\r\n\r\nif __name__ == '__main__':\r\n    main()\r\n"
    generate_speech_dataset:bool=Is Toggle Active(Target=Get EUW_ToggleButton_Text_GenerateSpeechDataset() -> (EUW_ToggleButton_Text_GenerateSpeechDataset:Tech Audio Tools Toggle Button Text Object Reference)) -> (ReturnValue=~)
    generate_transcriptions:bool=Is Toggle Active(Target=Get EUW_ToggleButton_Text_GenerateTranscriptions() -> (EUW_ToggleButton_Text_GenerateTranscriptions:Tech Audio Tools Toggle Button Text Object Reference)) -> (ReturnValue=~)
    generate_ratings:bool=Is Toggle Active(Target=Get EUW_ToggleButton_Text_GenerateRatings() -> (EUW_ToggleButton_Text_GenerateRatings:Tech Audio Tools Toggle Button Text Object Reference)) -> (ReturnValue=~)
    gather_transcriptions:bool=Is Toggle Active(Target=Get EUW_ToggleButton_Text_GatherTranscriptions() -> (EUW_ToggleButton_Text_GatherTranscriptions:Tech Audio Tools Toggle Button Text Object Reference)) -> (ReturnValue=~)
    filtered_configurations:Array of Strings=Get SelectedConfigurations()
    csv_file_path_dataset:String=GetSelectedFilePath(self=Get EUW_PathSelector_Dataset())
    csv_file_path_configurations:String=@n3
    csv_file_path_transcriptions_for_rating:String=@n3
    output_directory:String=GetSelectedFilePath(self=Get EUW_PathSelector_GeneratedWaveFileDirectory())
    samplerate:int=Select!(
     NewEnumerator0:int=8000
     NewEnumerator1:int=11025
     NewEnumerator2:int=16000
     NewEnumerator3:int=22050
     NewEnumerator4:int=24000
     NewEnumerator5:int=44100
     NewEnumerator6:int=48000
     Index:EAIVoiceGeneratorSampleRate Enum=Get SampleRate!() -> (SampleRate=NewEnumerator0)
    ) -> (ReturnValue:int)
    bitdepthindex:int=Select!(
     NewEnumerator0=1
     NewEnumerator1=2
     NewEnumerator2=3
     NewEnumerator3=4
     Index:EAIVoiceGeneratorBitDepth Enum=Get BitDepth!() -> (BitDepth=NewEnumerator0)
    ) -> (ReturnValue)
   ) -> (then, ReturnValue) {inputs [
     generate_speech_dataset
     generate_transcriptions
     generate_ratings
     gather_transcriptions
     filtered_configurations
     csv_file_path_dataset
     csv_file_path_configurations
     csv_file_path_transcriptions_for_rating
     output_directory
     samplerate
     bitdepthindex
    ]}
   > Branch(Condition=@n4)
   n5 = Format Text!(Format="LOCTEXT(\"At least one valid configuration must be selected\")") -> (Result)
   Show Message Dialog(execute=@n2.else, Title=Format Text(Format="LOCTEXT(\"Error\")"), Message=@n5, MessageType=Ok, DefaultValue=Ok, MessageCategory=Error)
   > AddErrorMessage!(execute, self, Message:Text=@n5) -> (then) {is=title K2Node_CallFunction}
   n6 = GenerateSourceAssets() {is=title K2Node_FunctionEntry}
  ]}
 HideWarningLabel {nodes [
   HideWarningLabel() {is=title K2Node_FunctionEntry}
   > SetVisibility(InVisibility=Hidden)
  ]}
 IsToggleActive {nodes [
   n1 = Inputs!() -> (Target:Tech Audio Tools Toggle Button Text Object Reference)
   Outputs!(ReturnValue:bool=Select!(
     Unchecked:bool=false
     Checked:bool=true
     Undetermined:bool=false
     Index:ECheckBoxState Enum=Get Checked State!(self:Tech Audio Tools Toggle Button Text Object Reference=@n1) -> (Checked State:ECheckBoxState Enum=Unchecked)
    ) -> (ReturnValue:bool))
  ]}
 RefreshOutputPanel {nodes [
   ClearChildren()
   > Clear(TargetArray:Array of EUW Clearable Info Label Object References=Get LoadingInfoLabels())
   > Clear(TargetArray:Array of Strings=Get ValidConfigurationNames())
   > n1 = Sequence()
   > Branch(Condition=Is Empty!(TargetArray=Get SelectedConfigurations()) -> (ReturnValue:bool))
   > AddWarningMessage(Message=Format Text(Format="LOCTEXT(\"No configurations selected\")"))
   n2 = Sequence(execute=@n1.then_1)
   n3 = Get EUW_PathSelector_Dataset()
   > Branch(Condition=IsEmpty(InString=GetSelectedFilePath(self=@n3)))
   > AddWarningMessage(Message=Format Text(Format="LOCTEXT(\"CSV path not set: \\'Script Dataset\\'\")"))
   n4 = Sequence(execute=@n2.then_1)
   n5 = Get EUW_PathSelector_Configuration()
   > Branch(Condition=IsEmpty(InString=GetSelectedFilePath(self=@n5)))
   > AddWarningMessage(Message=Format Text(Format="LOCTEXT(\"CSV path not set: \\'Voice Configurations\\'\")"))
   n6 = Sequence(execute=@n4.then_1)
   n7 = Get EUW_PathSelector_GeneratedWaveFileDirectory()
   > Branch(Condition=IsEmpty(InString=GetSelectedFilePath(self=@n7)))
   > AddWarningMessage(Message=Format Text(Format="LOCTEXT(\"Root output directory not set.\")"))
   n8 = Sequence(execute=@n6.then_1)
   > n9 = For Each Loop(Array:Array of Strings=Get SelectedConfigurations()) -> (Array Element:String)
   > n10 = Branch(Condition=IsEmpty(InString=@n9))
   > AddWarningMessage(Message=Format Text!(Format="LOCTEXT(\"Configuration \\'{i}\\' is empty.\")", i:int=@n9.Array Index) -> (Result))
   n11 = FindHeaderInCSVFile(execute=@n10.else, CsvFilePath=GetSelectedFilePath(self=Get EUW_PathSelector_Configuration()), HeaderName=@n9)
   > n12 = Branch(Condition=@n11)
   > Add(TargetArray=Get ValidConfigurationNames(), NewItem=@n9)
   AddWarningMessage(execute=@n12.else, Message="To Text (String)"(InString=@n11.OutputMsg))
   Sequence(execute=@n8.then_1)
   > n13 = Branch(Condition=HasAnyChildren!(self:Panel Widget Object Reference) -> (ReturnValue:bool))
   AddSuccessMessage!(execute=@n13.else, self, Message:Text="LOCTEXT(\"All Generation Settings are Valid\")") -> (then) {is=title K2Node_CallFunction}
   RefreshOutputPanel() {is=title K2Node_FunctionEntry}
  ]}
]
variables [
 BitDepth {type=EAIVoiceGeneratorBitDepth, default=16-Bit}
 CurrentAngle {type=number}
 FontSize {type=number, default=9}
 GenerationSettingsPreset {type=BP_AIVoiceGeneratorPreset_DataAsset_C, default=null}
 LoadingInfoLabels {type="array<EUW_ClearableInfoLabel_C>", default []}
 OutputIconSize {type=Vector2D, default=<20 20>}
 SampleRate {type=EAIVoiceGeneratorSampleRate, default=24 kHz}
 SelectedConfigurations {type="array<string>", default []}
 TextWidth {type=number, default=280}
 Tooltip_FilteredConfigurations_Body {type=string, default=A set of .wav files will be generated for each configuration in this list.}
 Tooltip_FilteredConfigurations_Title {type=string, default=Filtered Configurations}
 UpdateInterval {type=number, default=0.01}
 ValidConfigurationNames {type="array<string>", default []}
]
widgets {
 info {namedSlotCount=1, rootWidgetClass=&GridPanel, widgetCount=48}
 widgets [
  GridPanel_0 {widgetClassPath=&GridPanel, widget=GridPanel_0}
  BG_Main_A {
   widgetClassPath=&Image
   bIsVariable=true
   parent=GridPanel_0
   widget=BG_Main_A
   properties {
    brush {
     bIsDynamicallyLoaded=false
     drawAs=Image
     imageType=FullColor
     mirroring=NoMirror
     resourceName=None
     resourceObject=/DiscoveryBrowser/Art/Materials/Backgrounds/M_UI_Discovery_Background_FallGuys
     tiling=NoTile
     imageSize=<32 32>
     margin=<ltrb 0 0 0 0>
     outlineSettings {
      bUseBrushTransparency=false
      roundingType=HalfHeightRadius
      width=0
      color {colorUseRule=UseColor_Specified, specifiedColor=<rgba 0 0 0 0>}
      cornerRadii=<0 0 0 0>
     }
     tintColor {colorUseRule=UseColor_Specified, specifiedColor=<rgba 1 1 1 1>}
     uVRegion {bIsValid=false, max=<0 0>, min=<0 0>}
    }
    colorAndOpacity=<rgba 0.2103 0.1195 0.2292 1>
   }
   slot {class=GridSlot, properties {columnSpan=3, layer=-8, rowSpan=27}}
  }
  BG_Body {
   widgetClassPath=&Image
   bIsVariable=true
   parent=GridPanel_0
   widget=BG_Body
   properties {
    brush {
     bIsDynamicallyLoaded=false
     drawAs=Image
     imageType=FullColor
     mirroring=NoMirror
     resourceName=None
     resourceObject=/FortniteCoreUI/Frontend/StoreMain/Materials/M_UI_Screen_Vbucks_Background
     tiling=NoTile
     imageSize=<32 32>
     margin=<ltrb 0 0 0 0>
     outlineSettings {
      bUseBrushTransparency=false
      roundingType=HalfHeightRadius
      width=0
      color {colorUseRule=UseColor_Specified, specifiedColor=<rgba 0 0 0 0>}
      cornerRadii=<0 0 0 0>
     }
     tintColor {colorUseRule=UseColor_Specified, specifiedColor=<rgba 1 1 1 1>}
     uVRegion {bIsValid=false, max=<0 0>, min=<0 0>}
    }
    colorAndOpacity=<rgba 0.1042 0.1042 0.1042 1>
    renderTransformPivot=<0 0>
   }
   slot {class=GridSlot, properties {columnSpan=3, layer=-1, row=1, rowSpan=25}}
  }
  Spacer_Body_LeftMargin {
   widgetClassPath=&Spacer
   parent=GridPanel_0
   widget=Spacer_Body_LeftMargin
   properties {size=<20 20>}
   slot {class=GridSlot, properties {row=1}}
  }
  Spacer_Body_LeftMargin_1 {
   widgetClassPath=&Spacer
   parent=GridPanel_0
   widget=Spacer_Body_LeftMargin_1
   properties {size=<20 77.4881>}
   slot {class=GridSlot, properties {columnSpan=3}}
  }
  TextBlock_57 {
   widgetClassPath=&TextBlock
   parent=GridPanel_0
   widget=TextBlock_57
   properties {
    justification=Center
    text=ElevenLabs VO Generator
    font {
     bForceMonospaced=false
     bMaterialIsStencil=false
     fontMaterial=null
     fontObject=/LauncherSocial/BrutalType
     letterSpacing=100
     monospacedWidth=1
     size=31.5
     skewAmount=0.168
     typefaceFontName=Light
     outlineSettings {
      bApplyOutlineToDropShadows=false
      bMiteredCorners=false
      bSeparateFillAlpha=false
      outlineMaterial=null
      outlineSize=0
      outlineColor=<rgba 0 0 0 1>
     }
    }
   }
   slot {class=GridSlot, properties {column=1, verticalAlignment=VAlign_Center}}
  }
  Spacer_Body_RightMargin {
   widgetClassPath=&Spacer
   parent=GridPanel_0
   widget=Spacer_Body_RightMargin
   properties {size=<20 20>}
   slot {class=GridSlot, properties {column=2, row=1}}
  }
  Spacer_Body_RightMargin_1 {
   widgetClassPath=&Spacer
   parent=GridPanel_0
   widget=Spacer_Body_RightMargin_1
   properties {size=<20 20>}
   slot {class=GridSlot, properties {column=2, row=1}}
  }
  BG_Header_InputSettings {
   widgetClassPath=&Image
   bIsVariable=true
   parent=GridPanel_0
   widget=BG_Header_InputSettings
   properties {
    renderOpacity=0.6
    brush {
     bIsDynamicallyLoaded=false
     drawAs=Image
     imageType=FullColor
     mirroring=NoMirror
     resourceName=None
     resourceObject=/FortniteCoreUI/Athena/UI/Frontend/CosmeticItemCard/Materials/M_UI_ItemCard_Variant
     tiling=NoTile
     imageSize=<32 32>
     margin=<ltrb 0 0 0 0>
     outlineSettings {
      bUseBrushTransparency=false
      roundingType=HalfHeightRadius
      width=0
      color {colorUseRule=UseColor_Specified, specifiedColor=<rgba 0 0 0 0>}
      cornerRadii=<0 0 0 0>
     }
     tintColor {colorUseRule=UseColor_Specified, specifiedColor=<rgba 1 1 1 1>}
     uVRegion {bIsValid=false, max=<0 0>, min=<0 0>}
    }
    colorAndOpacity=<rgba 0.6934 0.6354 1 1>
   }
   slot {class=GridSlot, properties {column=1, row=2}}
  }
  TextBlock_2 {
   widgetClassPath=&TextBlock
   parent=GridPanel_0
   widget=TextBlock_2
   properties {
    justification=Center
    text=Generation Settings
    font {
     bForceMonospaced=false
     bMaterialIsStencil=false
     fontMaterial=null
     fontObject=&BrutalType
     letterSpacing=0
     monospacedWidth=1
     size=12
     skewAmount=0.15
     typefaceFontName=Regular
     outlineSettings {
      bApplyOutlineToDropShadows=false
      bMiteredCorners=false
      bSeparateFillAlpha=false
      outlineMaterial=null
      outlineSize=0
      outlineColor=<rgba 0 0 0 1>
     }
    }
   }
   slot {class=GridSlot, properties {column=1, horizontalAlignment=HAlign_Center, row=2, verticalAlignment=VAlign_Center}}
  }
  Spacer_InputSettings {
   widgetClassPath=&Spacer
   parent=GridPanel_0
   widget=Spacer_InputSettings
   properties {size=<1 20>}
   slot {class=GridSlot, properties {column=1, row=3}}
  }
  GridPanel_Checkboxes {
   widgetClassPath=&GridPanel
   parent=GridPanel_0
   widget=GridPanel_Checkboxes
   slot {class=GridSlot, properties {column=1, horizontalAlignment=HAlign_Left, row=8}}
  }
  EUW_ToggleButton_Text_GenerateSpeechDataset {
   widgetClassPath=/TechAudioToolsContent/Widgets/Editor/ToggleButtons/EUW_ToggleButton_Text
   bIsVariable=true
   parent=GridPanel_Checkboxes
   widget=EUW_ToggleButton_Text_GenerateSpeechDataset
   properties {
    checked State=Checked
    text=Generate Speech Dataset
    check Box Style {
     checkBoxType=CheckBox
     backgroundHoveredImage {
      bIsDynamicallyLoaded=false
      drawAs=RoundedBox
      imageType=NoImage
      mirroring=NoMirror
      resourceName=None
      resourceObject=null
      tiling=NoTile
      imageSize=<18 18>
      margin=<ltrb 0 0 0 0>
      outlineSettings {
       bUseBrushTransparency=false
       roundingType=FixedRadius
       width=1
       color {colorUseRule=UseColor_ColorTable, specifiedColor=<rgba 0 0 0 0>}
       cornerRadii=<3 3 3 3>
      }
      tintColor {colorUseRule=UseColor_Specified, specifiedColor=<rgba 0.0048 0.0048 0.0048 1>}
      uVRegion {bIsValid=false, max=<0 0>, min=<0 0>}
     }
     backgroundImage {
      bIsDynamicallyLoaded=false
      drawAs=RoundedBox
      imageType=NoImage
      mirroring=NoMirror
      resourceName=None
      resourceObject=null
      tiling=NoTile
      imageSize=<18 18>
      margin=<ltrb 0 0 0 0>
      outlineSettings {
       bUseBrushTransparency=false
       roundingType=FixedRadius
       width=1
       color {colorUseRule=UseColor_ColorTable, specifiedColor=<rgba 0 0 0 0>}
       cornerRadii=<3 3 3 3>
      }
      tintColor {colorUseRule=UseColor_Specified, specifiedColor=<rgba 0.0048 0.0048 0.0048 1>}
      uVRegion {bIsValid=false, max=<0 0>, min=<0 0>}
     }
     backgroundPressedImage {
      bIsDynamicallyLoaded=false
      drawAs=RoundedBox
      imageType=NoImage
      mirroring=NoMirror
      resourceName=None
      resourceObject=null
      tiling=NoTile
      imageSize=<18 18>
      margin=<ltrb 0 0 0 0>
      outlineSettings {
       bUseBrushTransparency=false
       roundingType=FixedRadius
       width=1
       color {colorUseRule=UseColor_ColorTable, specifiedColor=<rgba 0 0 0 0>}
       cornerRadii=<3 3 3 3>
      }
      tintColor {colorUseRule=UseColor_Specified, specifiedColor=<rgba 0.0048 0.0048 0.0048 1>}
      uVRegion {bIsValid=false, max=<0 0>, min=<0 0>}
     }
     borderBackgroundColor {colorUseRule=UseColor_Specified, specifiedColor=<rgba 1 1 1 1>}
     checkedForeground {colorUseRule=UseColor_Specified, specifiedColor=<rgba 1 1 1 1>}
     checkedHoveredForeground {colorUseRule=UseColor_Specified, specifiedColor=<rgba 1 1 1 1>}
     checkedHoveredImage {
      bIsDynamicallyLoaded=false
      drawAs=RoundedBox
      imageType=NoImage
      mirroring=NoMirror
      resourceName=None
      resourceObject=null
      tiling=NoTile
      imageSize=<0 0>
      margin=<ltrb 0 0 0 0>
      outlineSettings {
       bUseBrushTransparency=false
       roundingType=FixedRadius
       width=1
       color {colorUseRule=UseColor_Specified, specifiedColor=<rgba 0.006 0.006 0.006 1>}
       cornerRadii=<4 4 4 4>
      }
      tintColor {colorUseRule=UseColor_Specified, specifiedColor=<rgba 0.0044 0.2384 1 1>}
      uVRegion {bIsValid=false, max=<0 0>, min=<0 0>}
     }
     checkedImage {
      bIsDynamicallyLoaded=false
      drawAs=Image
      imageType=FullColor
      mirroring=NoMirror
      resourceName=None
      resourceObject=null
      tiling=NoTile
      imageSize=<18 18>
      margin=<ltrb 0 0 0 0>
      outlineSettings {
       bUseBrushTransparency=false
       roundingType=FixedRadius
       width=1
       color {colorUseRule=UseColor_Specified, specifiedColor=<rgba 0.006 0.006 0.006 1>}
       cornerRadii=<4 4 4 4>
      }
      tintColor {colorUseRule=UseColor_Specified, specifiedColor=<rgba 0 0.162 0.7454 1>}
      uVRegion {bIsValid=false, max=<0 0>, min=<0 0>}
     }
     checkedPressedForeground {colorUseRule=UseColor_Specified, specifiedColor=<rgba 1 1 1 1>}
     checkedPressedImage {
      bIsDynamicallyLoaded=false
      drawAs=RoundedBox
      imageType=NoImage
      mirroring=NoMirror
      resourceName=None
      resourceObject=null
      tiling=NoTile
      imageSize=<0 0>
      margin=<ltrb 0 0 0 0>
      outlineSettings {
       bUseBrushTransparency=false
       roundingType=FixedRadius
       width=1
       color {colorUseRule=UseColor_Specified, specifiedColor=<rgba 0.006 0.006 0.006 1>}
       cornerRadii=<4 4 4 4>
      }
      tintColor {colorUseRule=UseColor_Specified, specifiedColor=<rgba 0.0625 0.0625 0.0625 1>}
      uVRegion {bIsValid=false, max=<0 0>, min=<0 0>}
     }
     checkedSlateSound {resourceObject=null}
     foregroundColor {colorUseRule=UseColor_Specified, specifiedColor=<rgba 1 1 1 1>}
     hoveredForeground {colorUseRule=UseColor_Specified, specifiedColor=<rgba 1 1 1 1>}
     hoveredSlateSound {resourceObject=null}
     padding=<ltrb 8 5 8 5>
     pressedForeground {colorUseRule=UseColor_Specified, specifiedColor=<rgba 1 1 1 1>}
     uncheckedHoveredImage {
      bIsDynamicallyLoaded=false
      drawAs=RoundedBox
      imageType=NoImage
      mirroring=NoMirror
      resourceName=None
      resourceObject=null
      tiling=NoTile
      imageSize=<0 0>
      margin=<ltrb 0 0 0 0>
      outlineSettings {
       bUseBrushTransparency=false
       roundingType=FixedRadius
       width=1
       color {colorUseRule=UseColor_Specified, specifiedColor=<rgba 0.006 0.006 0.006 1>}
       cornerRadii=<4 4 4 4>
      }
      tintColor {colorUseRule=UseColor_Specified, specifiedColor=<rgba 0.0953 0.0953 0.0953 1>}
      uVRegion {bIsValid=false, max=<0 0>, min=<0 0>}
     }
     uncheckedImage {
      bIsDynamicallyLoaded=false
      drawAs=RoundedBox
      imageType=NoImage
      mirroring=NoMirror
      resourceName=None
      resourceObject=null
      tiling=NoTile
      imageSize=<0 0>
      margin=<ltrb 0 0 0 0>
      outlineSettings {
       bUseBrushTransparency=false
       roundingType=FixedRadius
       width=1
       color {colorUseRule=UseColor_Specified, specifiedColor=<rgba 0.006 0.006 0.006 1>}
       cornerRadii=<4 4 4 4>
      }
      tintColor {colorUseRule=UseColor_Specified, specifiedColor=<rgba 0.0284 0.0284 0.0284 1>}
      uVRegion {bIsValid=false, max=<0 0>, min=<0 0>}
     }
     uncheckedPressedImage {
      bIsDynamicallyLoaded=false
      drawAs=RoundedBox
      imageType=NoImage
      mirroring=NoMirror
      resourceName=None
      resourceObject=null
      tiling=NoTile
      imageSize=<0 0>
      margin=<ltrb 0 0 0 0>
      outlineSettings {
       bUseBrushTransparency=false
       roundingType=FixedRadius
       width=1
       color {colorUseRule=UseColor_Specified, specifiedColor=<rgba 0.006 0.006 0.006 1>}
       cornerRadii=<4 4 4 4>
      }
      tintColor {colorUseRule=UseColor_Specified, specifiedColor=<rgba 0.0395 0.0395 0.0395 1>}
      uVRegion {bIsValid=false, max=<0 0>, min=<0 0>}
     }
     uncheckedSlateSound {resourceObject=null}
     undeterminedForeground {colorUseRule=UseColor_Specified, specifiedColor=<rgba 1 1 1 1>}
     undeterminedHoveredImage {
      bIsDynamicallyLoaded=false
      drawAs=Image
      imageType=Vector
      mirroring=NoMirror
      resourceName=../../../Engine/Content/Slate/Starship/CoreWidgets/CheckBox/indeterminate.svg
      resourceObject=null
      tiling=NoTile
      imageSize=<16 16>
      margin=<ltrb 0 0 0 0>
      outlineSettings {
       bUseBrushTransparency=false
       roundingType=HalfHeightRadius
       width=0
       color {colorUseRule=UseColor_Specified, specifiedColor=<rgba 0 0 0 0>}
       cornerRadii=<0 0 0 0>
      }
      tintColor {colorUseRule=UseColor_Specified, specifiedColor=<rgba 0.0044 0.2384 1 1>}
      uVRegion {bIsValid=false, max=<0 0>, min=<0 0>}
     }
     undeterminedImage {
      bIsDynamicallyLoaded=false
      drawAs=Image
      imageType=Vector
      mirroring=NoMirror
      resourceName=../../../Engine/Content/Slate/Starship/CoreWidgets/CheckBox/indeterminate.svg
      resourceObject=null
      tiling=NoTile
      imageSize=<16 16>
      margin=<ltrb 0 0 0 0>
      outlineSettings {
       bUseBrushTransparency=false
       roundingType=HalfHeightRadius
       width=0
       color {colorUseRule=UseColor_Specified, specifiedColor=<rgba 0 0 0 0>}
       cornerRadii=<0 0 0 0>
      }
      tintColor {colorUseRule=UseColor_Specified, specifiedColor=<rgba 0 0.162 0.7454 1>}
      uVRegion {bIsValid=false, max=<0 0>, min=<0 0>}
     }
     undeterminedPressedImage {
      bIsDynamicallyLoaded=false
      drawAs=Image
      imageType=Vector
      mirroring=NoMirror
      resourceName=../../../Engine/Content/Slate/Starship/CoreWidgets/CheckBox/indeterminate.svg
      resourceObject=null
      tiling=NoTile
      imageSize=<16 16>
      margin=<ltrb 0 0 0 0>
      outlineSettings {
       bUseBrushTransparency=false
       roundingType=HalfHeightRadius
       width=0
       color {colorUseRule=UseColor_Specified, specifiedColor=<rgba 0 0 0 0>}
       cornerRadii=<0 0 0 0>
      }
      tintColor {colorUseRule=UseColor_Specified, specifiedColor=<rgba 0 0.0802 0.3515 1>}
      uVRegion {bIsValid=false, max=<0 0>, min=<0 0>}
     }
    }
   }
   slot {class=GridSlot, properties {horizontalAlignment=HAlign_Left, row=1}}
  }
  EUW_ToggleButton_Text_GenerateTranscriptions {
   widgetClassPath=&EUW_ToggleButton_Text
   bIsVariable=true
   parent=GridPanel_Checkboxes
   widget=EUW_ToggleButton_Text_GenerateTranscriptions
   properties {
    text=Generate Transcriptions
    check Box Style {
     checkBoxType=CheckBox
     backgroundHoveredImage {
      bIsDynamicallyLoaded=false
      drawAs=RoundedBox
      imageType=NoImage
      mirroring=NoMirror
      resourceName=None
      resourceObject=null
      tiling=NoTile
      imageSize=<18 18>
      margin=<ltrb 0 0 0 0>
      outlineSettings {
       bUseBrushTransparency=false
       roundingType=FixedRadius
       width=1
       color {colorUseRule=UseColor_ColorTable, specifiedColor=<rgba 0 0 0 0>}
       cornerRadii=<3 3 3 3>
      }
      tintColor {colorUseRule=UseColor_Specified, specifiedColor=<rgba 0.0048 0.0048 0.0048 1>}
      uVRegion {bIsValid=false, max=<0 0>, min=<0 0>}
     }
     backgroundImage {
      bIsDynamicallyLoaded=false
      drawAs=RoundedBox
      imageType=NoImage
      mirroring=NoMirror
      resourceName=None
      resourceObject=null
      tiling=NoTile
      imageSize=<18 18>
      margin=<ltrb 0 0 0 0>
      outlineSettings {
       bUseBrushTransparency=false
       roundingType=FixedRadius
       width=1
       color {colorUseRule=UseColor_ColorTable, specifiedColor=<rgba 0 0 0 0>}
       cornerRadii=<3 3 3 3>
      }
      tintColor {colorUseRule=UseColor_Specified, specifiedColor=<rgba 0.0048 0.0048 0.0048 1>}
      uVRegion {bIsValid=false, max=<0 0>, min=<0 0>}
     }
     backgroundPressedImage {
      bIsDynamicallyLoaded=false
      drawAs=RoundedBox
      imageType=NoImage
      mirroring=NoMirror
      resourceName=None
      resourceObject=null
      tiling=NoTile
      imageSize=<18 18>
      margin=<ltrb 0 0 0 0>
      outlineSettings {
       bUseBrushTransparency=false
       roundingType=FixedRadius
       width=1
       color {colorUseRule=UseColor_ColorTable, specifiedColor=<rgba 0 0 0 0>}
       cornerRadii=<3 3 3 3>
      }
      tintColor {colorUseRule=UseColor_Specified, specifiedColor=<rgba 0.0048 0.0048 0.0048 1>}
      uVRegion {bIsValid=false, max=<0 0>, min=<0 0>}
     }
     borderBackgroundColor {colorUseRule=UseColor_Specified, specifiedColor=<rgba 1 1 1 1>}
     checkedForeground {colorUseRule=UseColor_Specified, specifiedColor=<rgba 1 1 1 1>}
     checkedHoveredForeground {colorUseRule=UseColor_Specified, specifiedColor=<rgba 1 1 1 1>}
     checkedHoveredImage {
      bIsDynamicallyLoaded=false
      drawAs=RoundedBox
      imageType=NoImage
      mirroring=NoMirror
      resourceName=None
      resourceObject=null
      tiling=NoTile
      imageSize=<0 0>
      margin=<ltrb 0 0 0 0>
      outlineSettings {
       bUseBrushTransparency=false
       roundingType=FixedRadius
       width=1
       color {colorUseRule=UseColor_Specified, specifiedColor=<rgba 0.006 0.006 0.006 1>}
       cornerRadii=<4 4 4 4>
      }
      tintColor {colorUseRule=UseColor_Specified, specifiedColor=<rgba 0.0044 0.2384 1 1>}
      uVRegion {bIsValid=false, max=<0 0>, min=<0 0>}
     }
     checkedImage {
      bIsDynamicallyLoaded=false
      drawAs=RoundedBox
      imageType=FullColor
      mirroring=NoMirror
      resourceName=None
      resourceObject=null
      tiling=NoTile
      imageSize=<18 18>
      margin=<ltrb 0 0 0 0>
      outlineSettings {
       bUseBrushTransparency=false
       roundingType=FixedRadius
       width=1
       color {colorUseRule=UseColor_Specified, specifiedColor=<rgba 0.006 0.006 0.006 1>}
       cornerRadii=<4 4 4 4>
      }
      tintColor {colorUseRule=UseColor_Specified, specifiedColor=<rgba 0 0.162 0.7454 1>}
      uVRegion {bIsValid=false, max=<0 0>, min=<0 0>}
     }
     checkedPressedForeground {colorUseRule=UseColor_Specified, specifiedColor=<rgba 1 1 1 1>}
     checkedPressedImage {
      bIsDynamicallyLoaded=false
      drawAs=RoundedBox
      imageType=NoImage
      mirroring=NoMirror
      resourceName=None
      resourceObject=null
      tiling=NoTile
      imageSize=<0 0>
      margin=<ltrb 0 0 0 0>
      outlineSettings {
       bUseBrushTransparency=false
       roundingType=FixedRadius
       width=1
       color {colorUseRule=UseColor_Specified, specifiedColor=<rgba 0.006 0.006 0.006 1>}
       cornerRadii=<4 4 4 4>
      }
      tintColor {colorUseRule=UseColor_Specified, specifiedColor=<rgba 0.0625 0.0625 0.0625 1>}
      uVRegion {bIsValid=false, max=<0 0>, min=<0 0>}
     }
     checkedSlateSound {resourceObject=null}
     foregroundColor {colorUseRule=UseColor_Specified, specifiedColor=<rgba 1 1 1 1>}
     hoveredForeground {colorUseRule=UseColor_Specified, specifiedColor=<rgba 1 1 1 1>}
     hoveredSlateSound {resourceObject=null}
     padding=<ltrb 8 5 8 5>
     pressedForeground {colorUseRule=UseColor_Specified, specifiedColor=<rgba 1 1 1 1>}
     uncheckedHoveredImage {
      bIsDynamicallyLoaded=false
      drawAs=RoundedBox
      imageType=NoImage
      mirroring=NoMirror
      resourceName=None
      resourceObject=null
      tiling=NoTile
      imageSize=<0 0>
      margin=<ltrb 0 0 0 0>
      outlineSettings {
       bUseBrushTransparency=false
       roundingType=FixedRadius
       width=1
       color {colorUseRule=UseColor_Specified, specifiedColor=<rgba 0.006 0.006 0.006 1>}
       cornerRadii=<4 4 4 4>
      }
      tintColor {colorUseRule=UseColor_Specified, specifiedColor=<rgba 0.0953 0.0953 0.0953 1>}
      uVRegion {bIsValid=false, max=<0 0>, min=<0 0>}
     }
     uncheckedImage {
      bIsDynamicallyLoaded=false
      drawAs=RoundedBox
      imageType=NoImage
      mirroring=NoMirror
      resourceName=None
      resourceObject=null
      tiling=NoTile
      imageSize=<0 0>
      margin=<ltrb 0 0 0 0>
      outlineSettings {
       bUseBrushTransparency=false
       roundingType=FixedRadius
       width=1
       color {colorUseRule=UseColor_Specified, specifiedColor=<rgba 0.006 0.006 0.006 1>}
       cornerRadii=<4 4 4 4>
      }
      tintColor {colorUseRule=UseColor_Specified, specifiedColor=<rgba 0.0284 0.0284 0.0284 1>}
      uVRegion {bIsValid=false, max=<0 0>, min=<0 0>}
     }
     uncheckedPressedImage {
      bIsDynamicallyLoaded=false
      drawAs=RoundedBox
      imageType=NoImage
      mirroring=NoMirror
      resourceName=None
      resourceObject=null
      tiling=NoTile
      imageSize=<0 0>
      margin=<ltrb 0 0 0 0>
      outlineSettings {
       bUseBrushTransparency=false
       roundingType=FixedRadius
       width=1
       color {colorUseRule=UseColor_Specified, specifiedColor=<rgba 0.006 0.006 0.006 1>}
       cornerRadii=<4 4 4 4>
      }
      tintColor {colorUseRule=UseColor_Specified, specifiedColor=<rgba 0.0395 0.0395 0.0395 1>}
      uVRegion {bIsValid=false, max=<0 0>, min=<0 0>}
     }
     uncheckedSlateSound {resourceObject=null}
     undeterminedForeground {colorUseRule=UseColor_Specified, specifiedColor=<rgba 1 1 1 1>}
     undeterminedHoveredImage {
      bIsDynamicallyLoaded=false
      drawAs=Image
      imageType=Vector
      mirroring=NoMirror
      resourceName=../../../Engine/Content/Slate/Starship/CoreWidgets/CheckBox/indeterminate.svg
      resourceObject=null
      tiling=NoTile
      imageSize=<16 16>
      margin=<ltrb 0 0 0 0>
      outlineSettings {
       bUseBrushTransparency=false
       roundingType=HalfHeightRadius
       width=0
       color {colorUseRule=UseColor_Specified, specifiedColor=<rgba 0 0 0 0>}
       cornerRadii=<0 0 0 0>
      }
      tintColor {colorUseRule=UseColor_Specified, specifiedColor=<rgba 0.0044 0.2384 1 1>}
      uVRegion {bIsValid=false, max=<0 0>, min=<0 0>}
     }
     undeterminedImage {
      bIsDynamicallyLoaded=false
      drawAs=Image
      imageType=Vector
      mirroring=NoMirror
      resourceName=../../../Engine/Content/Slate/Starship/CoreWidgets/CheckBox/indeterminate.svg
      resourceObject=null
      tiling=NoTile
      imageSize=<16 16>
      margin=<ltrb 0 0 0 0>
      outlineSettings {
       bUseBrushTransparency=false
       roundingType=HalfHeightRadius
       width=0
       color {colorUseRule=UseColor_Specified, specifiedColor=<rgba 0 0 0 0>}
       cornerRadii=<0 0 0 0>
      }
      tintColor {colorUseRule=UseColor_Specified, specifiedColor=<rgba 0 0.162 0.7454 1>}
      uVRegion {bIsValid=false, max=<0 0>, min=<0 0>}
     }
     undeterminedPressedImage {
      bIsDynamicallyLoaded=false
      drawAs=Image
      imageType=Vector
      mirroring=NoMirror
      resourceName=../../../Engine/Content/Slate/Starship/CoreWidgets/CheckBox/indeterminate.svg
      resourceObject=null
      tiling=NoTile
      imageSize=<16 16>
      margin=<ltrb 0 0 0 0>
      outlineSettings {
       bUseBrushTransparency=false
       roundingType=HalfHeightRadius
       width=0
       color {colorUseRule=UseColor_Specified, specifiedColor=<rgba 0 0 0 0>}
       cornerRadii=<0 0 0 0>
      }
      tintColor {colorUseRule=UseColor_Specified, specifiedColor=<rgba 0 0.0802 0.3515 1>}
      uVRegion {bIsValid=false, max=<0 0>, min=<0 0>}
     }
    }
   }
   slot {class=GridSlot, properties {column=1, horizontalAlignment=HAlign_Left, row=1}}
  }
  EUW_ToggleButton_Text_GenerateRatings {
   widgetClassPath=&EUW_ToggleButton_Text
   bIsVariable=true
   parent=GridPanel_Checkboxes
   widget=EUW_ToggleButton_Text_GenerateRatings
   properties {
    text=Generate Ratings
    check Box Style {
     checkBoxType=CheckBox
     backgroundHoveredImage {
      bIsDynamicallyLoaded=false
      drawAs=RoundedBox
      imageType=NoImage
      mirroring=NoMirror
      resourceName=None
      resourceObject=null
      tiling=NoTile
      imageSize=<18 18>
      margin=<ltrb 0 0 0 0>
      outlineSettings {
       bUseBrushTransparency=false
       roundingType=FixedRadius
       width=1
       color {colorUseRule=UseColor_ColorTable, specifiedColor=<rgba 0 0 0 0>}
       cornerRadii=<3 3 3 3>
      }
      tintColor {colorUseRule=UseColor_Specified, specifiedColor=<rgba 0.0048 0.0048 0.0048 1>}
      uVRegion {bIsValid=false, max=<0 0>, min=<0 0>}
     }
     backgroundImage {
      bIsDynamicallyLoaded=false
      drawAs=RoundedBox
      imageType=NoImage
      mirroring=NoMirror
      resourceName=None
      resourceObject=null
      tiling=NoTile
      imageSize=<18 18>
      margin=<ltrb 0 0 0 0>
      outlineSettings {
       bUseBrushTransparency=false
       roundingType=FixedRadius
       width=1
       color {colorUseRule=UseColor_ColorTable, specifiedColor=<rgba 0 0 0 0>}
       cornerRadii=<3 3 3 3>
      }
      tintColor {colorUseRule=UseColor_Specified, specifiedColor=<rgba 0.0048 0.0048 0.0048 1>}
      uVRegion {bIsValid=false, max=<0 0>, min=<0 0>}
     }
     backgroundPressedImage {
      bIsDynamicallyLoaded=false
      drawAs=RoundedBox
      imageType=NoImage
      mirroring=NoMirror
      resourceName=None
      resourceObject=null
      tiling=NoTile
      imageSize=<18 18>
      margin=<ltrb 0 0 0 0>
      outlineSettings {
       bUseBrushTransparency=false
       roundingType=FixedRadius
       width=1
       color {colorUseRule=UseColor_ColorTable, specifiedColor=<rgba 0 0 0 0>}
       cornerRadii=<3 3 3 3>
      }
      tintColor {colorUseRule=UseColor_Specified, specifiedColor=<rgba 0.0048 0.0048 0.0048 1>}
      uVRegion {bIsValid=false, max=<0 0>, min=<0 0>}
     }
     borderBackgroundColor {colorUseRule=UseColor_Specified, specifiedColor=<rgba 1 1 1 1>}
     checkedForeground {colorUseRule=UseColor_Specified, specifiedColor=<rgba 1 1 1 1>}
     checkedHoveredForeground {colorUseRule=UseColor_Specified, specifiedColor=<rgba 1 1 1 1>}
     checkedHoveredImage {
      bIsDynamicallyLoaded=false
      drawAs=RoundedBox
      imageType=NoImage
      mirroring=NoMirror
      resourceName=None
      resourceObject=null
      tiling=NoTile
      imageSize=<0 0>
      margin=<ltrb 0 0 0 0>
      outlineSettings {
       bUseBrushTransparency=false
       roundingType=FixedRadius
       width=1
       color {colorUseRule=UseColor_Specified, specifiedColor=<rgba 0.006 0.006 0.006 1>}
       cornerRadii=<4 4 4 4>
      }
      tintColor {colorUseRule=UseColor_Specified, specifiedColor=<rgba 0.0044 0.2384 1 1>}
      uVRegion {bIsValid=false, max=<0 0>, min=<0 0>}
     }
     checkedImage {
      bIsDynamicallyLoaded=false
      drawAs=RoundedBox
      imageType=FullColor
      mirroring=NoMirror
      resourceName=None
      resourceObject=null
      tiling=NoTile
      imageSize=<18 18>
      margin=<ltrb 0 0 0 0>
      outlineSettings {
       bUseBrushTransparency=false
       roundingType=FixedRadius
       width=1
       color {colorUseRule=UseColor_Specified, specifiedColor=<rgba 0.006 0.006 0.006 1>}
       cornerRadii=<4 4 4 4>
      }
      tintColor {colorUseRule=UseColor_Specified, specifiedColor=<rgba 0 0.162 0.7454 1>}
      uVRegion {bIsValid=false, max=<0 0>, min=<0 0>}
     }
     checkedPressedForeground {colorUseRule=UseColor_Specified, specifiedColor=<rgba 1 1 1 1>}
     checkedPressedImage {
      bIsDynamicallyLoaded=false
      drawAs=RoundedBox
      imageType=NoImage
      mirroring=NoMirror
      resourceName=None
      resourceObject=null
      tiling=NoTile
      imageSize=<0 0>
      margin=<ltrb 0 0 0 0>
      outlineSettings {
       bUseBrushTransparency=false
       roundingType=FixedRadius
       width=1
       color {colorUseRule=UseColor_Specified, specifiedColor=<rgba 0.006 0.006 0.006 1>}
       cornerRadii=<4 4 4 4>
      }
      tintColor {colorUseRule=UseColor_Specified, specifiedColor=<rgba 0.0625 0.0625 0.0625 1>}
      uVRegion {bIsValid=false, max=<0 0>, min=<0 0>}
     }
     checkedSlateSound {resourceObject=null}
     foregroundColor {colorUseRule=UseColor_Specified, specifiedColor=<rgba 1 1 1 1>}
     hoveredForeground {colorUseRule=UseColor_Specified, specifiedColor=<rgba 1 1 1 1>}
     hoveredSlateSound {resourceObject=null}
     padding=<ltrb 8 5 8 5>
     pressedForeground {colorUseRule=UseColor_Specified, specifiedColor=<rgba 1 1 1 1>}
     uncheckedHoveredImage {
      bIsDynamicallyLoaded=false
      drawAs=RoundedBox
      imageType=NoImage
      mirroring=NoMirror
      resourceName=None
      resourceObject=null
      tiling=NoTile
      imageSize=<0 0>
      margin=<ltrb 0 0 0 0>
      outlineSettings {
       bUseBrushTransparency=false
       roundingType=FixedRadius
       width=1
       color {colorUseRule=UseColor_Specified, specifiedColor=<rgba 0.006 0.006 0.006 1>}
       cornerRadii=<4 4 4 4>
      }
      tintColor {colorUseRule=UseColor_Specified, specifiedColor=<rgba 0.0953 0.0953 0.0953 1>}
      uVRegion {bIsValid=false, max=<0 0>, min=<0 0>}
     }
     uncheckedImage {
      bIsDynamicallyLoaded=false
      drawAs=RoundedBox
      imageType=NoImage
      mirroring=NoMirror
      resourceName=None
      resourceObject=null
      tiling=NoTile
      imageSize=<0 0>
      margin=<ltrb 0 0 0 0>
      outlineSettings {
       bUseBrushTransparency=false
       roundingType=FixedRadius
       width=1
       color {colorUseRule=UseColor_Specified, specifiedColor=<rgba 0.006 0.006 0.006 1>}
       cornerRadii=<4 4 4 4>
      }
      tintColor {colorUseRule=UseColor_Specified, specifiedColor=<rgba 0.0284 0.0284 0.0284 1>}
      uVRegion {bIsValid=false, max=<0 0>, min=<0 0>}
     }
     uncheckedPressedImage {
      bIsDynamicallyLoaded=false
      drawAs=RoundedBox
      imageType=NoImage
      mirroring=NoMirror
      resourceName=None
      resourceObject=null
      tiling=NoTile
      imageSize=<0 0>
      margin=<ltrb 0 0 0 0>
      outlineSettings {
       bUseBrushTransparency=false
       roundingType=FixedRadius
       width=1
       color {colorUseRule=UseColor_Specified, specifiedColor=<rgba 0.006 0.006 0.006 1>}
       cornerRadii=<4 4 4 4>
      }
      tintColor {colorUseRule=UseColor_Specified, specifiedColor=<rgba 0.0395 0.0395 0.0395 1>}
      uVRegion {bIsValid=false, max=<0 0>, min=<0 0>}
     }
     uncheckedSlateSound {resourceObject=null}
     undeterminedForeground {colorUseRule=UseColor_Specified, specifiedColor=<rgba 1 1 1 1>}
     undeterminedHoveredImage {
      bIsDynamicallyLoaded=false
      drawAs=Image
      imageType=Vector
      mirroring=NoMirror
      resourceName=../../../Engine/Content/Slate/Starship/CoreWidgets/CheckBox/indeterminate.svg
      resourceObject=null
      tiling=NoTile
      imageSize=<16 16>
      margin=<ltrb 0 0 0 0>
      outlineSettings {
       bUseBrushTransparency=false
       roundingType=HalfHeightRadius
       width=0
       color {colorUseRule=UseColor_Specified, specifiedColor=<rgba 0 0 0 0>}
       cornerRadii=<0 0 0 0>
      }
      tintColor {colorUseRule=UseColor_Specified, specifiedColor=<rgba 0.0044 0.2384 1 1>}
      uVRegion {bIsValid=false, max=<0 0>, min=<0 0>}
     }
     undeterminedImage {
      bIsDynamicallyLoaded=false
      drawAs=Image
      imageType=Vector
      mirroring=NoMirror
      resourceName=../../../Engine/Content/Slate/Starship/CoreWidgets/CheckBox/indeterminate.svg
      resourceObject=null
      tiling=NoTile
      imageSize=<16 16>
      margin=<ltrb 0 0 0 0>
      outlineSettings {
       bUseBrushTransparency=false
       roundingType=HalfHeightRadius
       width=0
       color {colorUseRule=UseColor_Specified, specifiedColor=<rgba 0 0 0 0>}
       cornerRadii=<0 0 0 0>
      }
      tintColor {colorUseRule=UseColor_Specified, specifiedColor=<rgba 0 0.162 0.7454 1>}
      uVRegion {bIsValid=false, max=<0 0>, min=<0 0>}
     }
     undeterminedPressedImage {
      bIsDynamicallyLoaded=false
      drawAs=Image
      imageType=Vector
      mirroring=NoMirror
      resourceName=../../../Engine/Content/Slate/Starship/CoreWidgets/CheckBox/indeterminate.svg
      resourceObject=null
      tiling=NoTile
      imageSize=<16 16>
      margin=<ltrb 0 0 0 0>
      outlineSettings {
       bUseBrushTransparency=false
       roundingType=HalfHeightRadius
       width=0
       color {colorUseRule=UseColor_Specified, specifiedColor=<rgba 0 0 0 0>}
       cornerRadii=<0 0 0 0>
      }
      tintColor {colorUseRule=UseColor_Specified, specifiedColor=<rgba 0 0.0802 0.3515 1>}
      uVRegion {bIsValid=false, max=<0 0>, min=<0 0>}
     }
    }
   }
   slot {class=GridSlot, properties {column=2, horizontalAlignment=HAlign_Left, row=1}}
  }
  EUW_ToggleButton_Text_GatherTranscriptions {
   widgetClassPath=&EUW_ToggleButton_Text
   bIsVariable=true
   parent=GridPanel_Checkboxes
   widget=EUW_ToggleButton_Text_GatherTranscriptions
   properties {
    text=Gather Transcriptions
    check Box Style {
     checkBoxType=CheckBox
     backgroundHoveredImage {
      bIsDynamicallyLoaded=false
      drawAs=RoundedBox
      imageType=NoImage
      mirroring=NoMirror
      resourceName=None
      resourceObject=null
      tiling=NoTile
      imageSize=<18 18>
      margin=<ltrb 0 0 0 0>
      outlineSettings {
       bUseBrushTransparency=false
       roundingType=FixedRadius
       width=1
       color {colorUseRule=UseColor_ColorTable, specifiedColor=<rgba 0 0 0 0>}
       cornerRadii=<3 3 3 3>
      }
      tintColor {colorUseRule=UseColor_Specified, specifiedColor=<rgba 0.0048 0.0048 0.0048 1>}
      uVRegion {bIsValid=false, max=<0 0>, min=<0 0>}
     }
     backgroundImage {
      bIsDynamicallyLoaded=false
      drawAs=RoundedBox
      imageType=NoImage
      mirroring=NoMirror
      resourceName=None
      resourceObject=null
      tiling=NoTile
      imageSize=<18 18>
      margin=<ltrb 0 0 0 0>
      outlineSettings {
       bUseBrushTransparency=false
       roundingType=FixedRadius
       width=1
       color {colorUseRule=UseColor_ColorTable, specifiedColor=<rgba 0 0 0 0>}
       cornerRadii=<3 3 3 3>
      }
      tintColor {colorUseRule=UseColor_Specified, specifiedColor=<rgba 0.0048 0.0048 0.0048 1>}
      uVRegion {bIsValid=false, max=<0 0>, min=<0 0>}
     }
     backgroundPressedImage {
      bIsDynamicallyLoaded=false
      drawAs=RoundedBox
      imageType=NoImage
      mirroring=NoMirror
      resourceName=None
      resourceObject=null
      tiling=NoTile
      imageSize=<18 18>
      margin=<ltrb 0 0 0 0>
      outlineSettings {
       bUseBrushTransparency=false
       roundingType=FixedRadius
       width=1
       color {colorUseRule=UseColor_ColorTable, specifiedColor=<rgba 0 0 0 0>}
       cornerRadii=<3 3 3 3>
      }
      tintColor {colorUseRule=UseColor_Specified, specifiedColor=<rgba 0.0048 0.0048 0.0048 1>}
      uVRegion {bIsValid=false, max=<0 0>, min=<0 0>}
     }
     borderBackgroundColor {colorUseRule=UseColor_Specified, specifiedColor=<rgba 1 1 1 1>}
     checkedForeground {colorUseRule=UseColor_Specified, specifiedColor=<rgba 1 1 1 1>}
     checkedHoveredForeground {colorUseRule=UseColor_Specified, specifiedColor=<rgba 1 1 1 1>}
     checkedHoveredImage {
      bIsDynamicallyLoaded=false
      drawAs=RoundedBox
      imageType=NoImage
      mirroring=NoMirror
      resourceName=None
      resourceObject=null
      tiling=NoTile
      imageSize=<0 0>
      margin=<ltrb 0 0 0 0>
      outlineSettings {
       bUseBrushTransparency=false
       roundingType=FixedRadius
       width=1
       color {colorUseRule=UseColor_Specified, specifiedColor=<rgba 0.006 0.006 0.006 1>}
       cornerRadii=<4 4 4 4>
      }
      tintColor {colorUseRule=UseColor_Specified, specifiedColor=<rgba 0.0044 0.2384 1 1>}
      uVRegion {bIsValid=false, max=<0 0>, min=<0 0>}
     }
     checkedImage {
      bIsDynamicallyLoaded=false
      drawAs=RoundedBox
      imageType=FullColor
      mirroring=NoMirror
      resourceName=None
      resourceObject=null
      tiling=NoTile
      imageSize=<18 18>
      margin=<ltrb 0 0 0 0>
      outlineSettings {
       bUseBrushTransparency=false
       roundingType=FixedRadius
       width=1
       color {colorUseRule=UseColor_Specified, specifiedColor=<rgba 0.006 0.006 0.006 1>}
       cornerRadii=<4 4 4 4>
      }
      tintColor {colorUseRule=UseColor_Specified, specifiedColor=<rgba 0 0.162 0.7454 1>}
      uVRegion {bIsValid=false, max=<0 0>, min=<0 0>}
     }
     checkedPressedForeground {colorUseRule=UseColor_Specified, specifiedColor=<rgba 1 1 1 1>}
     checkedPressedImage {
      bIsDynamicallyLoaded=false
      drawAs=RoundedBox
      imageType=NoImage
      mirroring=NoMirror
      resourceName=None
      resourceObject=null
      tiling=NoTile
      imageSize=<0 0>
      margin=<ltrb 0 0 0 0>
      outlineSettings {
       bUseBrushTransparency=false
       roundingType=FixedRadius
       width=1
       color {colorUseRule=UseColor_Specified, specifiedColor=<rgba 0.006 0.006 0.006 1>}
       cornerRadii=<4 4 4 4>
      }
      tintColor {colorUseRule=UseColor_Specified, specifiedColor=<rgba 0.0625 0.0625 0.0625 1>}
      uVRegion {bIsValid=false, max=<0 0>, min=<0 0>}
     }
     checkedSlateSound {resourceObject=null}
     foregroundColor {colorUseRule=UseColor_Specified, specifiedColor=<rgba 1 1 1 1>}
     hoveredForeground {colorUseRule=UseColor_Specified, specifiedColor=<rgba 1 1 1 1>}
     hoveredSlateSound {resourceObject=null}
     padding=<ltrb 8 5 8 5>
     pressedForeground {colorUseRule=UseColor_Specified, specifiedColor=<rgba 1 1 1 1>}
     uncheckedHoveredImage {
      bIsDynamicallyLoaded=false
      drawAs=RoundedBox
      imageType=NoImage
      mirroring=NoMirror
      resourceName=None
      resourceObject=null
      tiling=NoTile
      imageSize=<0 0>
      margin=<ltrb 0 0 0 0>
      outlineSettings {
       bUseBrushTransparency=false
       roundingType=FixedRadius
       width=1
       color {colorUseRule=UseColor_Specified, specifiedColor=<rgba 0.006 0.006 0.006 1>}
       cornerRadii=<4 4 4 4>
      }
      tintColor {colorUseRule=UseColor_Specified, specifiedColor=<rgba 0.0953 0.0953 0.0953 1>}
      uVRegion {bIsValid=false, max=<0 0>, min=<0 0>}
     }
     uncheckedImage {
      bIsDynamicallyLoaded=false
      drawAs=RoundedBox
      imageType=NoImage
      mirroring=NoMirror
      resourceName=None
      resourceObject=null
      tiling=NoTile
      imageSize=<0 0>
      margin=<ltrb 0 0 0 0>
      outlineSettings {
       bUseBrushTransparency=false
       roundingType=FixedRadius
       width=1
       color {colorUseRule=UseColor_Specified, specifiedColor=<rgba 0.006 0.006 0.006 1>}
       cornerRadii=<4 4 4 4>
      }
      tintColor {colorUseRule=UseColor_Specified, specifiedColor=<rgba 0.0284 0.0284 0.0284 1>}
      uVRegion {bIsValid=false, max=<0 0>, min=<0 0>}
     }
     uncheckedPressedImage {
      bIsDynamicallyLoaded=false
      drawAs=RoundedBox
      imageType=NoImage
      mirroring=NoMirror
      resourceName=None
      resourceObject=null
      tiling=NoTile
      imageSize=<0 0>
      margin=<ltrb 0 0 0 0>
      outlineSettings {
       bUseBrushTransparency=false
       roundingType=FixedRadius
       width=1
       color {colorUseRule=UseColor_Specified, specifiedColor=<rgba 0.006 0.006 0.006 1>}
       cornerRadii=<4 4 4 4>
      }
      tintColor {colorUseRule=UseColor_Specified, specifiedColor=<rgba 0.0395 0.0395 0.0395 1>}
      uVRegion {bIsValid=false, max=<0 0>, min=<0 0>}
     }
     uncheckedSlateSound {resourceObject=null}
     undeterminedForeground {colorUseRule=UseColor_Specified, specifiedColor=<rgba 1 1 1 1>}
     undeterminedHoveredImage {
      bIsDynamicallyLoaded=false
      drawAs=Image
      imageType=Vector
      mirroring=NoMirror
      resourceName=../../../Engine/Content/Slate/Starship/CoreWidgets/CheckBox/indeterminate.svg
      resourceObject=null
      tiling=NoTile
      imageSize=<16 16>
      margin=<ltrb 0 0 0 0>
      outlineSettings {
       bUseBrushTransparency=false
       roundingType=HalfHeightRadius
       width=0
       color {colorUseRule=UseColor_Specified, specifiedColor=<rgba 0 0 0 0>}
       cornerRadii=<0 0 0 0>
      }
      tintColor {colorUseRule=UseColor_Specified, specifiedColor=<rgba 0.0044 0.2384 1 1>}
      uVRegion {bIsValid=false, max=<0 0>, min=<0 0>}
     }
     undeterminedImage {
      bIsDynamicallyLoaded=false
      drawAs=Image
      imageType=Vector
      mirroring=NoMirror
      resourceName=../../../Engine/Content/Slate/Starship/CoreWidgets/CheckBox/indeterminate.svg
      resourceObject=null
      tiling=NoTile
      imageSize=<16 16>
      margin=<ltrb 0 0 0 0>
      outlineSettings {
       bUseBrushTransparency=false
       roundingType=HalfHeightRadius
       width=0
       color {colorUseRule=UseColor_Specified, specifiedColor=<rgba 0 0 0 0>}
       cornerRadii=<0 0 0 0>
      }
      tintColor {colorUseRule=UseColor_Specified, specifiedColor=<rgba 0 0.162 0.7454 1>}
      uVRegion {bIsValid=false, max=<0 0>, min=<0 0>}
     }
     undeterminedPressedImage {
      bIsDynamicallyLoaded=false
      drawAs=Image
      imageType=Vector
      mirroring=NoMirror
      resourceName=../../../Engine/Content/Slate/Starship/CoreWidgets/CheckBox/indeterminate.svg
      resourceObject=null
      tiling=NoTile
      imageSize=<16 16>
      margin=<ltrb 0 0 0 0>
      outlineSettings {
       bUseBrushTransparency=false
       roundingType=HalfHeightRadius
       width=0
       color {colorUseRule=UseColor_Specified, specifiedColor=<rgba 0 0 0 0>}
       cornerRadii=<0 0 0 0>
      }
      tintColor {colorUseRule=UseColor_Specified, specifiedColor=<rgba 0 0.0802 0.3515 1>}
      uVRegion {bIsValid=false, max=<0 0>, min=<0 0>}
     }
    }
   }
   slot {class=GridSlot, properties {column=3, horizontalAlignment=HAlign_Left, row=1}}
  }
  Spacer_Configuration {
   widgetClassPath=&Spacer
   parent=GridPanel_0
   widget=Spacer_Configuration
   properties {size=<1 10>}
   slot {class=GridSlot, properties {column=1, row=5}}
  }
  SizeBox_0 {
   widgetClassPath=&SizeBox
   parent=GridPanel_0
   widget=SizeBox_0
   properties {bOverride_MaxDesiredHeight=true, bOverride_MinDesiredHeight=true, maxDesiredHeight=100, minDesiredHeight=100}
   slot {class=GridSlot, properties {column=1, row=16}}
  }
  DetailsView_Configurations {
   widgetClassPath=&DetailsView
   parent=SizeBox_0
   widget=DetailsView_Configurations
   properties {
    columnWidth=0.72
    toolTipText=The values entered in this list will be used to select voice model settings in the 'Voice Configurations' .csv file. Multiple configurations can be generated simultaneously
    propertiesToShow [SelectedConfigurations]
   }
   slot {class=SizeBoxSlot}
  }
  Spacer_DatasetPath {
   widgetClassPath=&Spacer
   parent=GridPanel_0
   widget=Spacer_DatasetPath
   properties {size=<1 10>}
   slot {class=GridSlot, properties {column=1, row=8}}
  }
  EUW_PathSelector_Dataset {
   widgetClassPath=&/EUW_PathSelector
   bIsVariable=true
   parent=GridPanel_0
   widget=EUW_PathSelector_Dataset
   properties {
    bOverride_Cursor=true
    textBoxHint=Full path to CSV file containing the lines to generate.
    textBoxWidth=510
    title="Script Dataset (.csv)"
    toolTipText="The 'Script Dataset' .csv file should contain a list of output filenames, paired with a transcription of the speech to generate."
    windowTitle=Select CSV File
   }
   slot {class=GridSlot, properties {column=1, row=10}}
  }
  Spacer_ConfigurationPath {
   widgetClassPath=&Spacer
   parent=GridPanel_0
   widget=Spacer_ConfigurationPath
   properties {size=<1 10>}
   slot {class=GridSlot, properties {column=1, row=9}}
  }
  EUW_PathSelector_Configuration {
   widgetClassPath=&EUW_PathSelector
   bIsVariable=true
   parent=GridPanel_0
   widget=EUW_PathSelector_Configuration
   properties {
    textBoxHint=Full path to CSV file containing AI voice model settings.
    textBoxWidth=510
    title="Voice Configurations (.csv)"
    toolTipText=The 'Voice Configurations' .csv file contains multiple configuration presets of the voice model settings. The 'Filtered Configurations' list above is used to select a subset of configurations from this list.
    windowTitle=Select CSV File
   }
   slot {class=GridSlot, properties {column=1, row=13}}
  }
  Spacer_OutputPath {
   widgetClassPath=&Spacer
   parent=GridPanel_0
   widget=Spacer_OutputPath
   properties {size=<1 10>}
   slot {class=GridSlot, properties {column=1, row=11}}
  }
  EUW_PathSelector_GeneratedWaveFileDirectory {
   widgetClassPath=&EUW_PathSelector
   bIsVariable=true
   parent=GridPanel_0
   widget=EUW_PathSelector_GeneratedWaveFileDirectory
   properties {
    fileSelectorType=System Directory
    fileTypeExtension=""
    fileTypeLabel=""
    textBoxHint=Directory where generated wave files should be stored.
    textBoxWidth=510
    title=Root Output Directory
    toolTipText=This directory will be used as the root folder for any generated .wav files. A subfolder will be created for each filtered configuration.
    windowTitle=Select Directory - Generated Wave Files
   }
   slot {class=GridSlot, properties {column=1, row=20}}
  }
  Spacer_OutputPath_1 {
   widgetClassPath=&Spacer
   parent=GridPanel_0
   widget=Spacer_OutputPath_1
   properties {size=<1 10>}
   slot {class=GridSlot, properties {column=1, row=13}}
  }
  Spacer_OutputPath_2 {
   widgetClassPath=&Spacer
   parent=GridPanel_0
   widget=Spacer_OutputPath_2
   properties {size=<1 10>}
   slot {class=GridSlot, properties {column=1, row=15}}
  }
  Spacer_OutputPath_3 {
   widgetClassPath=&Spacer
   parent=GridPanel_0
   widget=Spacer_OutputPath_3
   properties {size=<1 20>}
   slot {class=GridSlot, properties {column=1, row=17}}
  }
  Spacer_Footer_Body {
   widgetClassPath=&Spacer
   parent=GridPanel_0
   widget=Spacer_Footer_Body
   properties {size=<1 20>}
   slot {class=GridSlot, properties {column=1, row=5}}
  }
  Border_201 {
   widgetClassPath=&Border
   parent=GridPanel_0
   widget=Border_201
   properties {brushColor=<rgba 0.0058 0.0632 0.1719 1>, renderTransform {angle=0, scale=<1 0.05>, shear=<0 0>, translation=<0 0>}}
   slot {class=GridSlot, properties {column=1, row=5}}
  }
  Border_2 {
   widgetClassPath=&Border
   parent=GridPanel_0
   widget=Border_2
   properties {brushColor=<rgba 0.0058 0.0632 0.1719 1>, renderTransform {angle=0, scale=<1 0.05>, shear=<0 0>, translation=<0 0>}}
   slot {class=GridSlot, properties {column=1, row=3}}
  }
  Border {
   widgetClassPath=&Border
   parent=GridPanel_0
   widget=Border
   properties {brushColor=<rgba 0.0058 0.0632 0.1719 1>, renderTransform {angle=0, scale=<1 0.05>, shear=<0 0>, translation=<0 0>}}
   slot {class=GridSlot, properties {column=1, row=17}}
  }
  Spacer_Footer_Main {widgetClassPath=&Spacer, parent=GridPanel_0, widget=Spacer_Footer_Main, slot {class=GridSlot, properties {row=19}}}
  Spacer_60 {
   widgetClassPath=&Spacer
   parent=GridPanel_0
   widget=Spacer_60
   properties {size=<580 5>}
   slot {class=GridSlot, properties {column=1, row=18}}
  }
  Spacer_3 {
   widgetClassPath=&Spacer
   parent=GridPanel_0
   widget=Spacer_3
   properties {size=<580 10>}
   slot {class=GridSlot, properties {column=1, row=21}}
  }
  Spacer_4 {
   widgetClassPath=&Spacer
   parent=GridPanel_0
   widget=Spacer_4
   properties {size=<580 20>}
   slot {class=GridSlot, properties {column=1, row=23}}
  }
  Spacer {
   widgetClassPath=&Spacer
   parent=GridPanel_0
   widget=Spacer
   properties {size=<580 5>}
   slot {class=GridSlot, properties {column=1, row=7}}
  }
  Spacer_1 {
   widgetClassPath=&Spacer
   parent=GridPanel_0
   widget=Spacer_1
   properties {size=<580 20>}
   slot {class=GridSlot, properties {column=1, row=25}}
  }
  Spacer_2 {
   widgetClassPath=&Spacer
   parent=GridPanel_0
   widget=Spacer_2
   properties {size=<580 30>}
   slot {class=GridSlot, properties {column=1, row=26}}
  }
  GridPanel_112 {
   widgetClassPath=&GridPanel
   parent=GridPanel_0
   widget=GridPanel_112
   slot {class=GridSlot, properties {columnSpan=5, horizontalAlignment=HAlign_Center}}
  }
  SizeBox_3 {
   widgetClassPath=&SizeBox
   parent=GridPanel_0
   widget=SizeBox_3
   properties {bOverride_MinDesiredHeight=true, bOverride_MinDesiredWidth=true, minDesiredHeight=47.9713, minDesiredWidth=415.3492}
   slot {class=GridSlot, properties {column=1, horizontalAlignment=HAlign_Center, row=24}}
  }
  EUW_Button_Primary_GenerateSourceAssets {
   widgetClassPath=/TechAudioToolsContent/Widgets/Editor/Buttons/EUW_Button_Primary
   bIsVariable=true
   parent=SizeBox_3
   widget=EUW_Button_Primary_GenerateSourceAssets
   slot {class=SizeBoxSlot}
  }
  TextBlock_164 {
   widgetClassPath=&TextBlock
   namedSlotHost=EUW_Button_Primary_GenerateSourceAssets
   widget=TextBlock_164
   properties {
    text=Generate Source Assets
    font {
     bForceMonospaced=false
     bMaterialIsStencil=false
     fontMaterial=null
     fontObject=&Roboto
     letterSpacing=0
     monospacedWidth=1
     size=15
     skewAmount=0
     typefaceFontName=Bold
     outlineSettings {
      bApplyOutlineToDropShadows=false
      bMiteredCorners=false
      bSeparateFillAlpha=false
      outlineMaterial=null
      outlineSize=0
      outlineColor=<rgba 0 0 0 1>
     }
    }
   }
  }
  DetailsView_SampleRate {
   widgetClassPath=&DetailsView
   parent=GridPanel_0
   widget=DetailsView_SampleRate
   properties {columnWidth=0.7208, propertiesToShow [SampleRate]}
   slot {class=GridSlot, properties {column=1, row=22}}
  }
  SizeBox_4 {
   widgetClassPath=&SizeBox
   parent=GridPanel_0
   widget=SizeBox_4
   properties {bOverride_MaxDesiredWidth=true, bOverride_MinDesiredHeight=true, maxDesiredWidth=500, minDesiredHeight=60, minDesiredWidth=500}
   slot {class=GridSlot, properties {column=1, row=4}}
  }
  DetailsView_Preset {
   widgetClassPath=&DetailsView
   parent=SizeBox_4
   widget=DetailsView_Preset
   properties {
    columnWidth=0.72
    toolTipText=The values entered in this list will be used to select voice model settings in the 'Voice Configurations' .csv file. Multiple configurations can be generated simultaneously
    propertiesToShow [GenerationSettingsPreset]
   }
   slot {class=SizeBoxSlot}
  }
  TextBlock_1 {
   widgetClassPath=&TextBlock
   parent=GridPanel_0
   widget=TextBlock_1
   properties {
    text=OUTPUT SETTINGS
    colorAndOpacity {colorUseRule=UseColor_Specified, specifiedColor=<rgba 0.1581 0.2818 0.4323 1>}
    font {
     bForceMonospaced=false
     bMaterialIsStencil=false
     fontMaterial=null
     fontObject=&Roboto
     letterSpacing=0
     monospacedWidth=1
     size=7.5
     skewAmount=0
     typefaceFontName=Bold
     outlineSettings {
      bApplyOutlineToDropShadows=false
      bMiteredCorners=false
      bSeparateFillAlpha=false
      outlineMaterial=null
      outlineSize=0
      outlineColor=<rgba 0 0 0 1>
     }
    }
   }
   slot {class=GridSlot, properties {column=1, row=19, nudge=<5 0>, padding=<ltrb 0 0 0 4>}}
  }
  TextBlock {
   widgetClassPath=&TextBlock
   parent=GridPanel_0
   widget=TextBlock
   properties {
    text=INPUT SETTINGS
    colorAndOpacity {colorUseRule=UseColor_Specified, specifiedColor=<rgba 0.1581 0.2818 0.4323 1>}
    font {
     bForceMonospaced=false
     bMaterialIsStencil=false
     fontMaterial=null
     fontObject=&Roboto
     letterSpacing=0
     monospacedWidth=1
     size=7.5
     skewAmount=0
     typefaceFontName=Bold
     outlineSettings {
      bApplyOutlineToDropShadows=false
      bMiteredCorners=false
      bSeparateFillAlpha=false
      outlineMaterial=null
      outlineSize=0
      outlineColor=<rgba 0 0 0 1>
     }
    }
   }
   slot {class=GridSlot, properties {column=1, row=6, nudge=<5 0>, padding=<ltrb 0 0 0 4>}}
  }
 ]
}
